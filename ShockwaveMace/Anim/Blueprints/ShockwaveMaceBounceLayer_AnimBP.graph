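# Player anim layer Blueprint that plays the Shockwave Mace's full-body bounce and fall animations.
# An inertialized FullBodyOverride state machine cycles Bounce, fall transition and FallLoop on falling state, exiting from the first two into BounceExit.
# Each update derives bounce and exit flags from pawn gameplay tags and a BlendOut curve, while notifies track state flags.
# The grounded launch-event and ground-exit branches have no incoming execution connection.
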AnimBlueprint /ShockwaveMace/Anim/Blueprints/ShockwaveMaceBounceLayer_AnimBP
target_skeleton=&Fortnite_M_Avg_Player_Skeleton
graphs [
 AnimGraph {nodes [Output Pose() {layerGroup=None}]}
 CharacterSkeletalControlPostLegIK {nodes [
   Input Pose() {name=InPosePostLegIK}
   Output Pose() {layerGroup=PlayerOverride}
  ]}
 EventGraph {nodes [
   Event BlueprintUpdateAnimation()
   > n1 = Sequence!(execute) -> (then_0, then_1, then_2:Exec)
   > n2 = Get!(execute:Exec) -> (then:Exec, else:Exec, FortPlayerPawn:Fort Player Pawn Object Reference)
   > n3 = Branch(Condition=HasMatchingGameplayTag!(self:Gameplay Tag Asset Interface Interface=@n2, TagToCheck:Gameplay Tag Structure=(TagName="Abilities.ShockwaveMace.Bounce")) -> (ReturnValue:bool))
   > Set IsMaceBounce(IsMaceBounce:bool=true) -> (Output_Get:bool)
   Set IsMaceBounce(execute=[@n3.else, @n2.else], IsMaceBounce=false)
   n4 = Get(execute=@n1.then_2)
   n5 = GetCurveValue(CurveName=BlendOut)
   > n6 = Branch(Condition=AND Boolean(A=HasMatchingGameplayTag(self=@n4, TagToCheck=(TagName="Abilities.ShockwaveMace.Bounce.Input.Released")), B="float > float"(A=@n5)))
   > Set IsMaceBounceExit(IsMaceBounceExit:bool=true) -> (Output_Get:bool)
   Set IsMaceBounceExit(execute=@n6.else, IsMaceBounceExit=false)
   AnimNotify_EnterIntroBounce() {is=title K2Node_Event}
   > Set EnterIntroBounce(EnterIntroBounce:bool=true) -> (Output_Get:bool)
   AnimNotify_EnterBounce() {is=title K2Node_Event}
   > Set IsBounceState(IsBounceState:bool=true) -> (Output_Get:bool)
   > Set IsBounceToFallState(IsBounceToFallState:bool=false) -> (Output_Get:bool)
   > Set IsFallLoopState(IsFallLoopState:bool=false) -> (Output_Get:bool)
   AnimNotify_EnterBounce_tr_Fall() {is=title K2Node_Event}
   > Set IsBounceToFallState(IsBounceToFallState=true)
   > Set IsBounceState(IsBounceState=false)
   > Set IsFallLoopState(IsFallLoopState=false)
   AnimNotify_EnterFallLoop() {is=title K2Node_Event}
   > Set IsFallLoopState(IsFallLoopState=true)
   > Set IsBounceState(IsBounceState=false)
   > Set IsBounceToFallState(IsBounceToFallState=false)
   AnimNotify_ExitBounceExit() {is=title K2Node_Event}
   > Set IsFallLoopState(IsFallLoopState=false)
   > Set IsBounceState(IsBounceState=false)
   > Set IsBounceToFallState(IsBounceToFallState=false)
   Branch(Condition=AND Boolean(A=Get EnterIntroBounce() -> (EnterIntroBounce:bool), B=Get bIsOnGround() -> (bIsOnGround:bool)))
   > SendGameplayEventToActor!(
    execute
    Actor:Actor Object Reference=TryGetPawnOwner()
    EventTag:Gameplay Tag Structure=(TagName="Event.ShockwaveMace.Trigger.Bounce.Launch")
    Payload:Gameplay Event Data Structure
   ) -> (then)
   n7 = Get IsMaceBounce() -> (IsMaceBounce:bool)
   n8 = Branch(Condition=AND Boolean(A=Get bIsOnGround(), B=NOT Boolean(A=@n7)))
   > Set IsMaceGroundExit(IsMaceGroundExit:bool=true) -> (Output_Get:bool)
   Set IsMaceGroundExit(execute=@n8.else, IsMaceGroundExit=false)
   NOT Boolean(A=Get IsMaceBounce())
  ]}
 FullBodyOverride {nodes [
   Input Pose() {name=PassThroughFullBodyPose}
   AnimGraphNode_StateMachine_0 = MaceBounceStateMachine() {is=title AnimGraphNode_StateMachine, bAllowConduitEntryStates=true}
   Output Pose(Result=Inertialization!(Source:Pose Link Structure=@AnimGraphNode_StateMachine_0) -> (Pose:Pose Link Structure) {is=title AnimGraphNode_Inertialization}) {layerGroup=PlayerOverride}
  ]}
 MaceBounceStateMachine {
  entry=Bounce
  path=FullBodyOverride.AnimGraphNode_StateMachine_0.MaceBounceStateMachine
  ids=id
  nodes [
   BounceToFallExit() {is=title AnimStateAliasNode, also {aliases [Bounce, Bounce_tr_Fall]}}
   AnimStateNode_0 = BounceExit() {is=title AnimStateNode}
   AnimStateNode_1 = Bounce() {
    is=title AnimStateNode
    stateEntered {guid=#1, notifyName=EnterBounce}
    stateFullyBlended {guid=#2}
    stateLeft {guid=#3}
   }
   AnimStateNode_2 = Bounce_tr_Fall() {is=title AnimStateNode, stateEntered {guid=#4, notifyName=EnterBounce_tr_Fall}}
   AnimStateNode_3 = FallLoop() {
    is=title AnimStateNode
    stateEntered {guid=#5, notifyName=EnterFallLoop}
    stateFullyBlended {guid=#6}
    stateLeft {guid=#7}
   }
   AnimStateTransitionNode_0 = AnimStateTransitionNode() {crossfadeDuration=0.4, logicType=TLT_Inertialization}
   AnimStateTransitionNode_1 = AnimStateTransitionNode() {crossfadeDuration=0.3, logicType=TLT_Inertialization}
   AnimStateTransitionNode_2 = AnimStateTransitionNode() {bAutomaticRuleBasedOnSequencePlayerInState=true, crossfadeDuration=0.3, logicType=TLT_Inertialization}
   AnimStateTransitionNode_4 = AnimStateTransitionNode() {crossfadeDuration=0, logicType=TLT_Inertialization}
  ]
  links [
   {
    dst=Bounce_tr_Fall
    rule=FullBodyOverride.AnimGraphNode_StateMachine_0.MaceBounceStateMachine.AnimStateTransitionNode_1.Transition
    src=Bounce
    via=AnimStateTransitionNode_1
   }
   {
    dst=BounceExit
    rule=FullBodyOverride.AnimGraphNode_StateMachine_0.MaceBounceStateMachine.AnimStateTransitionNode_0.Transition
    src=BounceToFallExit
    via=AnimStateTransitionNode_0
   }
   {
    dst=FallLoop
    rule=FullBodyOverride.AnimGraphNode_StateMachine_0.MaceBounceStateMachine.AnimStateTransitionNode_2.Transition
    src=Bounce_tr_Fall
    via=AnimStateTransitionNode_2
   }
   {
    dst=Bounce
    rule=FullBodyOverride.AnimGraphNode_StateMachine_0.MaceBounceStateMachine.AnimStateTransitionNode_4.Transition
    src=FallLoop
    via=AnimStateTransitionNode_4
   }
  ]
 }
 BounceExit {
  path=FullBodyOverride.AnimGraphNode_StateMachine_0.MaceBounceStateMachine.AnimStateNode_0.BounceExit
  nodes [Output Animation Pose(Result=Sequence Player 'ShockWaveMace_Transition_Apex_To_Fall'() {bLoopAnimation=false, sequence=/ShockwaveMace/Anim/BasicHammer/ShockWaveMace_Transition_Apex_To_Fall})]
 }
 Bounce {
  path=FullBodyOverride.AnimGraphNode_StateMachine_0.MaceBounceStateMachine.AnimStateNode_1.Bounce
  nodes [Output Animation Pose(Result=Sequence Player 'ShockWaveMace_Bounce_Apex'() {bLoopAnimation=false, sequence=/ShockwaveMace/Anim/BasicHammer/ShockWaveMace_Bounce_Apex})]
 }
 Bounce_tr_Fall {
  path=FullBodyOverride.AnimGraphNode_StateMachine_0.MaceBounceStateMachine.AnimStateNode_2.Bounce_tr_Fall
  nodes [Output Animation Pose(Result=Sequence Player 'ShockWaveMace_Bounce_Loop_Into'() {bLoopAnimation=false, sequence=/ShockwaveMace/Anim/BasicHammer/ShockWaveMace_Bounce_Loop_Into})]
 }
 FallLoop {
  path=FullBodyOverride.AnimGraphNode_StateMachine_0.MaceBounceStateMachine.AnimStateNode_3.FallLoop
  nodes [Output Animation Pose(Result=Sequence Player 'ShockWaveMace_Bounce_Loop'() {sequence=/ShockwaveMace/Anim/BasicHammer/ShockWaveMace_Bounce_Loop})]
 }
 Transition {
  path=FullBodyOverride.AnimGraphNode_StateMachine_0.MaceBounceStateMachine.AnimStateTransitionNode_0.Transition
  nodes [Result(bCanEnterTransition=Get IsMaceBounceExit() -> (IsMaceBounceExit:bool))]
 }
 Transition {
  path=FullBodyOverride.AnimGraphNode_StateMachine_0.MaceBounceStateMachine.AnimStateTransitionNode_1.Transition
  nodes [Result(bCanEnterTransition=Get bIsFalling())]
 }
 Transition {path=FullBodyOverride.AnimGraphNode_StateMachine_0.MaceBounceStateMachine.AnimStateTransitionNode_2.Transition, nodes [Result()]}
 Transition {
  path=FullBodyOverride.AnimGraphNode_StateMachine_0.MaceBounceStateMachine.AnimStateTransitionNode_4.Transition
  nodes [Result(bCanEnterTransition=NOT Boolean(A=Get bIsFalling()))]
 }
 PreLegIK {nodes [
   Input Pose()
   Output Pose() {layerGroup=PlayerOverride}
  ]}
 UpperBody {nodes [
   Input Pose() {name=InPoseUpperBody}
   Output Pose() {layerGroup=PlayerOverride}
  ]}
]
variables [
 EnterIntroBounce {type=boolean}
 IsBounceState {type=boolean}
 IsBounceToFallState {type=boolean}
 IsFallLoopState {type=boolean}
 IsMaceBounce {type=boolean}
 IsMaceBounceExit {type=boolean}
 IsMaceGroundExit {type=boolean}
]
