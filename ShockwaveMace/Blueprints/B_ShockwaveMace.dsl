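; Shockwave Mace weapon Blueprint that presents magazine ammo as charge through material, Niagara and audio feedback.
; Ammo changes drive gain/loss sounds, a local gameplay cue on gain, an FX restart pulse and a PowerLevel parameter on ambient audio.
; A client-only timer eases the charge into custom primitive data and shows steam at zero.
; Ambient audio follows equip, visibility and blocking tags; ground contact timestamps the material and shows a sound indicator.
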(event EventBeginPlay
  (bind _asfort_player_pawn (Utilities|Casting|CastToFortPlayerPawn (Game|GetInstigator)))
  (Vehicle|BindEventtoOnVehicleStateChangedEvent _asfort_player_pawn (delegate Custom|VehicleStateChanged))
  (Variables|Default|SetAmmoChargeValue (Math|Conversions|ToFloat(Integer) (FortWeapon|GetMagazineAmmoCount))))

(event Custom|VehicleStateChanged
  (params (FortPlayerPawn :type Fort%20Player%20Pawn%20Object%20Reference) (NewVehicle :type Actor%20Object%20Reference) (OldVehicle :type Actor%20Object%20Reference))
  (CallFunction|SetAmbientAudioState true))

(event EventOnWeaponVisibilityChanged
  (params (bVisible :type Boolean) (bSetForLocalControllerOnly :type Boolean))
  (if (Variables|Default|GetbAmmoInitialized)
    (CallFunction|SetAmbientAudioState bVisible)))

(event FortWeapon|EventOnEquippedAndAttached
  (EventDispatchers|BindEventtoOnLocalAmmoChanged (EventDispatchers|CreateEvent :!event Custom|AmmoChanged))
  (CallFunction|UpdateChargeState)
  (Variables|Default|SetbEquipped true)
  (Components|Activation|Activate (Variables|Default|GetNS_ShockwaveMace_Energized) true))

(event FortWeapon|EventOnUnEquippedAndDetached
  (EventDispatchers|UnbindEventfromOnLocalAmmoChanged (EventDispatchers|CreateEvent :!event Custom|AmmoChanged))
  (Variables|Default|SetbEquipped false)
  (CallFunction|SetAmbientAudioState false)
  (Components|Activation|Deactivate (Variables|Default|GetNS_ShockwaveMace_Energized)))

(event Custom|AmmoChanged
  (params (LocalCount :type Integer) (LocalRemaining :type Integer))
  (bind _returnvalue (FortWeapon|GetMagazineAmmoCount))
  (CallFunction|UpdateChargeState)
  ;[1 1 1] Try to avoid issue when first equipping the weapon and ammo hasn't been cached yet, calling increase/decrease ammo inadvertently
  (if (Variables|Default|GetbAmmoInitialized)
    (if (!= (Variables|Default|GetCachedAmmoCount) _returnvalue)
      (if (> (Variables|Default|GetCachedAmmoCount) _returnvalue)
        (CallFunction|AmmoDecreased)
        (else
          (CallFunction|AmmoIncreased)
          (CallFunction|SetAmbientAudioState true)
          (Niagara|SetNiagaraVariable(Bool) (Variables|Default|GetNS_ShockwaveMace_Energized) "Steam" false)))
      (Variables|Default|SetCachedAmmoCount (FortWeapon|GetMagazineAmmoCount)))
    (else
      (bind _output_get (Variables|Default|SetCachedAmmoCount (FortWeapon|GetMagazineAmmoCount)))
      (Variables|Default|SetMaxAmmo _output_get)
      (Variables|Default|SetbAmmoInitialized true)
      (CallFunction|SetAmbientAudioState true))))

(event Custom|UpdateChargeState
  (if (not (Networking|IsServer))
    (Variables|Default|SetTargetAmmoCount (FortWeapon|GetMagazineAmmoCount))
    (bind _returnvalue (Utilities|Time|SetTimerbyEvent (delegate Custom|TickAmmoChargeValue) 0.016 true))
    (Variables|Default|SetAmmoValueTimerHandle _returnvalue)))

(event Custom|TickAmmoChargeValue
  (bind _output_get (Variables|Default|SetAmmoChargeValue (Math|Interpolation|FInterptoConstant (Variables|Default|GetAmmoChargeValue) (Math|Conversions|ToFloat(Integer) (Variables|Default|GetTargetAmmoCount)) 0.016 4.0)))
  (Rendering|Material|SetCustomPrimitiveDataFloat (Variables|Visual|GetWeaponMesh) 0 _output_get)
  (Effects|Components|ParticleSystem|SetFloatParameter (Variables|Default|GetNS_ShockwaveMace_Energized) "AmmoCount" (Math|Conversions|ToFloat(Integer) (FortWeapon|GetMagazineAmmoCount)))
  (if (Math|Float|NearlyEqual(Float) (Variables|Default|GetAmmoChargeValue) (Math|Conversions|ToFloat(Integer) (Variables|Default|GetTargetAmmoCount)) 0.0001)
    (Utilities|Time|ClearandInvalidateTimerbyHandle (Variables|Default|GetAmmoValueTimerHandle))
    (if (== (Variables|Default|GetTargetAmmoCount) 0)
      (Niagara|SetNiagaraVariable(Bool) (Variables|Default|GetNS_ShockwaveMace_Energized) "Steam" true))
    (else
      (Niagara|SetNiagaraVariable(Bool) (Variables|Default|GetNS_ShockwaveMace_Energized) "Steam" false))))

(event Custom|AmmoIncreased
  (bind _ns_shockwavemace_energized (Variables|Default|GetNS_ShockwaveMace_Energized))
  (CallFunction|AmmoChargedGC)
  (Utilities|FlowControl|Sequence
    (:then_0
      (Audio|Parameter|SetIntegerParameter (Variables|Default|GetAmbientAudioComp) "PowerLevel" (FortWeapon|GetMagazineAmmoCount) :!class /Script/AudioExtensions.AudioParameterControllerInterface)
      (bind _sb_ammoincreased (Variables|Audio|GetSB_AmmoIncreased)
        (:then
          (Audio|SpawnSoundAttached :Sound _sb_ammoincreased :AttachToComponent (Variables|Visual|GetWeaponMesh) :AttachPointName "Muzzle" :LocationType "SnapToTarget" :bStopWhenAttachedToDestroyed true :VolumeMultiplier 1.0 :PitchMultiplier 1.0 :StartTime 0.0 :bAutoDestroy true)
          (Audio|Parameter|SetIntegerParameter (AudioParameter|GetAudioParameterComponent) "PowerLevel" (FortWeapon|GetMagazineAmmoCount) :!class /Script/AudioExtensions.AudioParameterControllerInterface))
        (:else)))
    (:then_1
      (Components|Activation|SetActive _ns_shockwavemace_energized false false) ; Charge added, turn off for a tick
      (Utilities|FlowControl|Delay 0.05)
      (Components|Activation|SetActive _ns_shockwavemace_energized true true) ; FX activate to indicate charge added
    )))

(event Custom|AmmoDecreased
  (Audio|Parameter|SetIntegerParameter (Variables|Default|GetAmbientAudioComp) "PowerLevel" (FortWeapon|GetMagazineAmmoCount) :!class /Script/AudioExtensions.AudioParameterControllerInterface)
  (bind _sb_ammodecreased (Variables|Audio|GetSB_AmmoDecreased)
    (:then
      (Audio|SpawnSoundAttached :Sound _sb_ammodecreased :AttachToComponent (Variables|Visual|GetWeaponMesh) :AttachPointName "Muzzle" :LocationType "SnapToTarget" :bStopWhenAttachedToDestroyed true :VolumeMultiplier 1.0 :PitchMultiplier 1.0 :StartTime 0.0 :bAutoDestroy true))
    (:else)))

(event Custom|HandleGroundContact
  (params (GameTime :type Float%20%28double-precision%29))
  (Rendering|Material|SetCustomPrimitiveDataFloat (Variables|Visual|GetWeaponMesh) 1 (Utilities|Time|GetGameTimeinSeconds))
  (Default|CallOnGroundContact)
  (SoundIndicator|ShowSoundIndicator (Variables|SoundIndicator|GetSoundIndicatorComponent) (Transformation|GetActorLocation (Game|GetInstigator)) 0 1.0 0))

(fn AmmoChargedGC
  (bind _asfort_pawn (Utilities|Casting|CastToFortPawn (Variables|Actor|GetInstigator)))
  (if (Pawn|IsLocallyControlled _asfort_pawn)
    (GameplayCue|ExecuteGameplayCue%28Burst%29Local _asfort_pawn (Variables|Default|GetAmmoChargeCue) (Ability|GameplayCue|MakeGameplayCueParameters 0.0))))

(fn SetAmbientAudioState
  (params (bEnabled :type Boolean))
  (if bEnabled
    (if (and (not (GameplayTags|HasAnyMatchingGameplayTags (Utilities|Casting|CastToFortPlayerPawn (Variables|Actor|GetInstigator)) (Variables|Default|GetTC_ActionsThatPreventIdleAudio))) (Variables|Default|GetbEquipped))
      (Variables|Default|GetAmbientAudioComp
        (:then
          _)
        (:else
          (bind _returnvalue (Audio|SpawnSoundAttached (Variables|Audio|GetSB_MaceAmbient_Loop) (Variables|Visual|GetWeaponMesh)))
          (bind _output_get (Variables|Default|SetAmbientAudioComp _returnvalue))))
      (Audio|Parameter|SetIntegerParameter _output_get "PowerLevel" (FortWeapon|GetMagazineAmmoCount) :!class /Script/AudioExtensions.AudioParameterControllerInterface))
    (else
      (Audio|Components|Audio|FadeOut (Variables|Default|GetAmbientAudioComp) 0.1 0.0 "Linear" :!class /Script/Engine.AudioComponent))))

(fn ConstructionScript
  (|CallParentFunction)
  (Rendering|Material|CreateDynamicMaterialInstance (Variables|Visual|GetWeaponMesh) 0 "/ShockwaveMace/Materials/MI_Shockwave_Mace.MI_Shockwave_Mace" "None" :!class /Script/Engine.PrimitiveComponent))
