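# Additive unlit Niagara mesh-particle material for the Shockwave Mace weapon charge-up glow.
# Two panned plasma-cloud samples are summed, squared and shaped by vertical UV gradients, then masked by inverted depth fade to glow near intersecting geometry, except on Low quality.
# A fresnel from scene world normals is added on non-mobile paths.
# Dynamic parameters tune exponents and intensities; particle color tints the result before MF_FX_EmissiveAdjustments.
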Material /ShockwaveMace/FX/Materials/M_ShockwaveMace_ChargeUp_Weapon
graph {
 properties {
  bUsedWithNiagaraMeshParticles=true
  blendMode=BLEND_Additive
  pixelDepthOffsetMode=PDOM_Legacy
  shadingModel=MSM_Unlit
  usedShadingModels=Unlit
 }
 nodes [
  ComponentMask(None=TextureCoordinate()) {g=true}
  n1 = ParticleColor()
  n2 = DynamicParameter() {defaultValue=<rgba 5 2 10 10>, paramNames [Texture Exp, Fade Exp, Fresnel Multiplier, Intensity Multiplier]}
  n3 = TextureCoordinate()
  n4 = Multiply(A=@n3, B=Constant2Vector() {g=2, r=0.5})
  n5 = Multiply(A=@n3, B=Constant2Vector() {g=1.3, r=0.6})
  n6 = Add(
   A=TextureSample(UVs=Panner(Coordinate=@n4) {speedX=0.2, speedY=1}) {texture=/Game/Weapons/WeaponSkins/Wraps/Textures/PlasmaClouds/T_Wrap_PlasmaClouds__M}.G
   B=TextureSample(UVs=Panner(Coordinate=@n5) {speedX=-0.3, speedY=0.8}) {texture=&T_Wrap_PlasmaClouds__M}.G
  )
  n7 = OneMinus(None=ComponentMask(None=TextureCoordinate()) {g=true})
  n8 = DynamicParameter() {defaultValue=<rgba 5 2 10 10>, paramNames [Texture Exp, Fade Exp, Fresnel Multiplier, Intensity Multiplier]}
  n9 = LinearInterpolate(A=Multiply(A=@n6, B=@n6), Alpha=Power(Base=@n7, Exp=@n8.Texture Exp) {constExponent=5})
  n10 = Power(Base=@n7, Exp=@n8.Fade Exp)
  n11 = OneMinus(None=DepthFade() {fadeDistanceDefault=30})
  n12 = Constant() {r=1}
  n13 = Multiply(
   A=Multiply(A=@n1.A, B=Multiply(A=@n9, B=@n10))
   B=LinearInterpolate(
    A=QualitySwitch(Default=@n11, Low=@n12, High=@n11, Medium=@n11, Epic=@n11)
    Alpha=DynamicParameter() {parameterIndex=1, defaultValue=<rgba 0 0 0 1>, paramNames [Debug Mesh, "", "", ""]}.Debug Mesh
   )
  )
  n14 = Multiply(A=@n2.Intensity Multiplier, B=@n13)
  n15 = DotProduct(A=CameraVectorWS(), B=ComponentMask(None=SceneTexture() {sceneTextureId=PPI_WorldNormal}.Color) {b=true, g=true, r=true})
  n16 = Multiply(A=Power(Base=OneMinus(None=@n15)) {constExponent=3}, B=@n2.Fresnel Multiplier)
  n17 = ShadingPathSwitch(Default=@n16, Deferred=@n16, Mobile=Constant()) {desc=Migrated from FeatureLevelSwitch}
  MaterialFunctionCall(Emissive=Multiply(A=@n1.RGB, B=Add(A=@n14, B=@n17))) {materialFunction=&MF_FX_EmissiveAdjustments}
  Reroute(None=@n13)
  ScalarParameter() {defaultValue=10, parameterName=Intensity Multiplier}
 ]
}
