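# Niagara system for the Shockwave Mace's bounce effect: a short electric shockwave burst with dust and debris.
# Burst emitters layer dust, arcs, streaks, ground-wave and wind meshes, and an expanding sphere with a light.
# Emitter-age curves cool the bright magenta flash, while dynamic material parameters animate shaders, partly from system ripple variables.
# Rocks fall under gravity and bounce, disabling collision once a speed-and-age check fires; rocks and wind are distance-culled.
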NiagaraSystem /ShockwaveMace/FX/NS_ShockwaveMace_Bounce
dependencies {
 usedDataInterfaces [
  &NiagaraDataInterfaceVectorCurve
  &NiagaraDataInterfaceCurve
  &NiagaraDataInterfaceColorCurve
  &NiagaraDataInterfaceVector2DCurve
 ]
 usedDynamicInputs []
 usedModules [
  &SystemState
  &self:SystemSpawnScript.SystemScriptSource.SystemScriptGraph.NiagaraNodeAssignment_0.SetVariables_#1
 ]
}
emitters {
 Dust {
  emitterUpdateScript [
   &EmitterState(Life Cycle Mode=::NewEnumerator1, Loop Behavior=::NewEnumerator1, Loop Duration=1, UseLoopDelay=true)
   &SpawnBurst_Instantaneous(Spawn Count=11, Use Spawn Probability=false)
  ]
  particleSpawnScript [
   &FN_SetSpawnGroup(Normalize Spawn Group=true, Number of Spawn Groups=1, Particles Per Group=10)
   &InitializeParticle(
    Color Maximum=<rgba 0.8385 0.9347 1 0.8>
    Color Minimum=<rgba 0.8125 0.8227 1 0.6>
    Color Mode=::NewEnumerator2
    Lifetime Max=1
    Lifetime Min=0.6
    Lifetime Mode=::NewEnumerator1
    Sprite Size Mode=::NewEnumerator1
    Uniform Sprite Size Max=200
    Uniform Sprite Size Min=140
    UsePositionOffset=false
    Write Material Random=true
   ) {module=InitializeParticle001}
   &TorusLocation(
    Align Torus With Axis=true
    Handle Radius=5
    Large Radius=15
    MaskSpawnGroup=false
    Torus Distribution Mode:ENiagaraTorusDistributionMode=::NewEnumerator1
    U Position:NiagaraFloat=&Lerp_Float(A=0.05, B=0.95, Alpha=$Particles.Normalized Spawn Group)
   )
   &AddVelocityFromPoint(
    Module.Use Falloff Distance=false
    Origin Offset=<10 0 0>
    Velocity Strength=&RandomRangeFloat(Minimum=1800, Maximum=2400)
   )
  ]
  particleUpdateScript [
   &ParticleState()
   &Drag(Drag=&RandomRangeFloat(Minimum=5, Maximum=6), Use Linear Drag=true, Use Rotational Drag=false)
   &SpriteRotationRate(Rotation Rate=&Multiply_Float(
     A=&RandomRangeFloat(Minimum=-3, Maximum=3)
     B=&FloatFromCurve(
      FloatCurve=DataInterface() {propertyValues {Curve {keys [
          {arriveTangent=-4.0107, interpMode=RCIM_Cubic, leaveTangent=-4.0107, tangentMode=RCTM_User, time=-0.0046, value=1.0113}
          {arriveTangent=-0.2496, interpMode=RCIM_Cubic, leaveTangent=-0.2496, tangentMode=RCTM_User, time=0.5, value=0.16}
         ]}}}
      CurveIndex=$Emitter.Age
      Scale Curve=150
     )
    ))
   &WindForce(Air Resistance=0.4, Wind Speed=<0 0 1>, Wind Speed Scale=&RandomRangeFloat(Minimum=400, Maximum=900))
   &ScaleSpriteSize(Scale Factor=&Vector2DFromFloat(Value=&FloatFromCurve(FloatCurve=DataInterface() {propertyValues {Curve {keys [
          {arriveTangent=2.3678, interpMode=RCIM_Cubic, leaveTangent=2.3678, tangentMode=RCTM_User, time=0, value=0.2}
          {arriveTangent=0.7401, interpMode=RCIM_Cubic, leaveTangent=0.7401, tangentMode=RCTM_User, time=0.3, value=0.7}
          {arriveTangent=0.2539, interpMode=RCIM_Cubic, leaveTangent=0.2539, tangentMode=RCTM_User, time=0.9978, value=1}
         ]}}})))
   &SolveForcesAndVelocity(Clamp Velocity=false, Limit Acceleration=false)
   &ScaleColor(
    Scale Alpha=&FloatFromCurve(
     FloatCurve=DataInterface() {propertyValues {Curve {keys [
         {arriveTangent=5.4511, interpMode=RCIM_Cubic, leaveTangent=5.4511, tangentMode=RCTM_User, time=0, value=0.4}
         {arriveTangent=-1.4813, interpMode=RCIM_Cubic, leaveTangent=-1.4813, tangentMode=RCTM_User, time=0.2, value=1}
         {arriveTangent=-0.484, interpMode=RCIM_Cubic, leaveTangent=-0.484, tangentMode=RCTM_User, time=1, value=0}
        ]}}}
     Scale Curve=&RandomRangeFloat(Minimum=0.5, Maximum=0.3)
    )
    Scale RGB=&MakeVectorFromLinearColorRGB(Color=&ColorFromCurve(
      ColorCurve=DataInterface() {propertyValues {
        AlphaCurve {keys []}
        BlueCurve {keys [<at 0.0485 6 RCIM_Cubic>, {arriveTangent=-19.0486, interpMode=RCIM_Cubic, leaveTangent=-19.0486, time=0.1696, value=2}, <at 0.3295 0.6458>]}
        GreenCurve {keys [
          <at 0.0485 1.018 RCIM_Cubic>
          {arriveTangent=-2.2041, interpMode=RCIM_Cubic, leaveTangent=-2.2041, time=0.1696, value=0.4856}
          <at 0.3295 0.3985>
         ]}
        RedCurve {keys [
          <at 0.0485 5.153 RCIM_Cubic>
          {arriveTangent=-16.8034, interpMode=RCIM_Cubic, leaveTangent=-16.8034, time=0.1696, value=1.4424}
          <at 0.3295 0.4299>
         ]}
       }}
      CurveIndex=$Emitter.Age
     ))
    ScaleA=true
    ScaleRGB=true
   )
   &DynamicMaterialParameters(
    Index 0 Param 1=&Add_Float(
     A=$Particles.RandomFloat
     B=&FloatFromCurve(
      FloatCurve=DataInterface() {propertyValues {Curve {keys [
          {arriveTangent=2.7533, interpMode=RCIM_Cubic, leaveTangent=1.8958, tangentMode=RCTM_User, time=0, value=1}
          {arriveTangent=0.8292, interpMode=RCIM_Cubic, leaveTangent=0.8292, tangentMode=RCTM_User, time=0.321, value=1.3453}
          {arriveTangent=0.0529, interpMode=RCIM_Cubic, leaveTangent=0.2021, tangentMode=RCTM_User, time=1, value=1.5}
         ]}}}
      CurveIndex=&Add_Float(A=&Multiply_Float(A=$Particles.Initial.MaterialRandom, B=0.5), B=$Particles.NormalizedAge)
     )
    )
    Index 0 Param 2=&FloatFromCurve(FloatCurve=DataInterface() {propertyValues {Curve {keys [
         {arriveTangent=0.0453, interpMode=RCIM_Cubic, tangentMode=RCTM_Break, time=0, value=-1}
         {arriveTangent=0.0604, interpMode=RCIM_Cubic, leaveTangent=0.0604, tangentMode=RCTM_User, time=1, value=-1}
        ]}}})
    Index 0 Param 3=&FloatFromCurve(FloatCurve=DataInterface() {propertyValues {Curve {keys [
         {arriveTangent=-53.9199, interpMode=RCIM_Cubic, leaveTangent=-8.9727, tangentMode=RCTM_User, time=0, value=1}
         {arriveTangent=0.2347, interpMode=RCIM_Cubic, leaveTangent=0.2723, tangentMode=RCTM_Break, time=0.25, value=0}
        ]}}})
    Index 0 Param 4=&FloatFromCurve(FloatCurve=DataInterface() {propertyValues {Curve {keys [<at 0.5 1>,<at 1 0>]}}})
    Index 2 Param 1=&FloatFromCurve(FloatCurve=DataInterface() {propertyValues {Curve {keys [
         {interpMode=RCIM_Cubic, leaveTangent=4.2485, tangentMode=RCTM_Break, time=0, value=0}
         {arriveTangent=1.0548, interpMode=RCIM_Cubic, leaveTangent=1.0548, tangentMode=RCTM_User, time=0.17, value=0.5715}
         <at 1 1>
        ]}}})
    Index 2 Param 2=&FloatFromCurve(
     FloatCurve=DataInterface() {propertyValues {Curve {keys [
         {arriveTangent=0.9316, interpMode=RCIM_Cubic, leaveTangent=0.9316, tangentMode=RCTM_User, time=0, value=0.1}
         {arriveTangent=0.0259, interpMode=RCIM_Cubic, leaveTangent=0.0259, tangentMode=RCTM_User, time=1, value=0.4}
        ]}}}
     Scale Curve=&RandomRangeFloat(Minimum=0.2, Maximum=0.7)
    )
    Index 2 Param 3=&FloatFromCurve(FloatCurve=DataInterface() {propertyValues {Curve {keys [
         {interpMode=RCIM_Cubic, leaveTangent=2.3064, tangentMode=RCTM_Break, time=0, value=1}
         {interpMode=RCIM_Cubic, leaveTangent=11.1233, tangentMode=RCTM_Break, time=1, value=2}
        ]}}})
    Index 2 Param 4=&FloatFromCurve(
     FloatCurve=DataInterface() {propertyValues {Curve {keys [
         {arriveTangent=-0.6867, interpMode=RCIM_Cubic, leaveTangent=-0.6867, tangentMode=RCTM_User, time=0, value=1}
         {interpMode=RCIM_Cubic, leaveTangent=-1.2956, tangentMode=RCTM_Break, time=1, value=0.75}
        ]}}}
     Scale Curve=1.6
    )
    Module.Param0WriteEnabled=true
    Module.Param0WriteEnabled2=true
    Module.Param1WriteEnabled=false
    Module.Param1WriteEnabled2=true
    Module.Param2WriteEnabled=true
    Module.Param2WriteEnabled2=true
    Module.Param3WriteEnabled=true
    Module.Param3WriteEnabled2=true
    Write Parameter Index 2=true
   )
   &CameraOffset(Camera Offset Amount=/Niagara/DynamicInputs/Math/CalculateParticleRadius(
     Radius Calculation Type:ENiagaraCollisionRadiusOptions=::NewEnumerator0
     Particle Radius:NiagaraFloat=1
     Mesh Dimensions:Vector3f=<1 1 1>
     Method for Calculating Particle Radius:ENiagaraCalculateRadiusOptions=::NewEnumerator2
    ))
  ]
  rendererClasses [&NiagaraSpriteRendererProperties]
  renderers [
   {rendererClass=&NiagaraSpriteRendererProperties}
   {rendererClass=&NiagaraSpriteRendererProperties, rendererIndex=1}
  ]
 }
 ElectricArcs {
  emitterUpdateScript [
   &EmitterState(Life Cycle Mode=::NewEnumerator1, Loop Behavior=::NewEnumerator1, Loop Duration=0.25, UseLoopDelay=false)
   &SpawnBurst_Instantaneous(Spawn Count=2, Use Spawn Probability=false, UseLoopCountLimit=false)
   &SpawnBurst_Instantaneous(Spawn Count=2, Spawn Time=0.04, Use Spawn Probability=false, UseLoopCountLimit=false) {module=SpawnBurst_Instantaneous002}
   &SpawnBurst_Instantaneous(Spawn Count=4, Spawn Time=0.07, Use Spawn Probability=false, UseLoopCountLimit=false) {module=SpawnBurst_Instantaneous001}
  ]
  particleSpawnScript [
   /Niagara/Modules/Spawn/Initialization/InitializeParticle(
    Color=&RandomRangeLinearColor(Minimum=<rgba 1.9477 1.0986 4 1>, Maximum=<rgba 4.19 2.4994 6 1>)
    Lifetime=&UniformRangedFloat(Minimum=0.1, Maximum=0.15)
    Module.Write Color=true
    Module.Write Lifetime=true
    Module.Write Mass=true
    Module.Write Material Random=true
    Module.Write Position=false
    Module.Write Scale=false
    Module.Write SpriteRotation=true
    Module.Write SpriteSize=true
    Position=&SimulationPosition()
    Sprite Rotation=&UniformRangedFloat(Minimum=0, Maximum=360)
    Sprite Size=&UniformRangedVector2D(Minimum=<70 70>, Maximum=<150 150>)
   )
   &DynamicMaterialParameters(
    Index 0 Param 1=$Particles.MaterialRandom
    Index 0 Param 2=&Lerp_Float(A=0.7, B=1.25, Alpha=$Particles.MaterialRandom)
    Index 0 Param 4=1
    Module.Param0WriteEnabled=true
    Module.Param1WriteEnabled=true
    Module.Param2WriteEnabled=false
    Module.Param3WriteEnabled=true
   )
   &CameraOffset(Camera Offset Amount=25)
   &ShapeLocation(
    Cylinder Height=10
    Cylinder Radius=70
    MaskSpawnGroup=false
    Override Seed=false
    Shape Primitive=::NewEnumerator1
    Surface Only=true
    Use Endcaps In Surface Only Mode=false
    UseOwnerScale=true
   )
   &AddVelocity(Module.Use Falloff Distance=false, Origin Offset=<-50 0 0>, Override Seed=false, Velocity Mode=::NewEnumerator1, Velocity Speed=800)
  ]
  particleUpdateScript [
   &ParticleState()
   &SpriteRotationRate(Rotation Rate=&RandomRangeFloat(Minimum=-200, Maximum=200))
   &DynamicMaterialParameters(
    Index 0 Param 3=&FloatFromCurve(FloatCurve=DataInterface() {propertyValues {Curve {keys [<at 0.5 0 RCIM_Cubic>, {arriveTangent=3.7979, interpMode=RCIM_Cubic, leaveTangent=3.7979, tangentMode=RCTM_User, time=1, value=1}]}}})
    Index 1 Param 2=0.7
    Index 1 Param 3=$System.RipplePan
    Index 1 Param 4=500
    Module.Param0WriteEnabled=false
    Module.Param0WriteEnabled1=true
    Module.Param1WriteEnabled=false
    Module.Param1WriteEnabled1=true
    Module.Param2WriteEnabled=true
    Module.Param2WriteEnabled1=true
    Module.Param3WriteEnabled=false
    Module.Param3WriteEnabled1=true
    Write Parameter Index 1=true
   ) {module=DynamicMaterialParameters001}
   &ScaleSpriteSize(Scale Factor=&Vector2DFromFloat(Value=&FloatFromCurve(FloatCurve=DataInterface() {propertyValues {Curve {keys [{arriveTangent=0.3147, interpMode=RCIM_Cubic, leaveTangent=0.3147, tangentMode=RCTM_User, time=0, value=0.8}, <at 1 1>]}}})))
   &Drag(Drag=3, Use Linear Drag=true, Use Rotational Drag=false)
   &SolveForcesAndVelocity(Clamp Velocity=false, Limit Acceleration=false)
   &ScaleColor(
    Scale RGB=&MakeVectorFromLinearColorRGB(Color=&ColorFromCurve(
      ColorCurve=DataInterface() {propertyValues {
        AlphaCurve {keys []}
        BlueCurve {keys [
          <at 0.0485 1.8072 RCIM_Cubic>
          {arriveTangent=-2.3921, interpMode=RCIM_Cubic, leaveTangent=-2.3921, time=0.2983, value=1.6918}
          <at 0.534 0.6458>
         ]}
        GreenCurve {keys [
          <at 0.0485 0.4164 RCIM_Cubic>
          {arriveTangent=0.1445, interpMode=RCIM_Cubic, leaveTangent=0.1445, time=0.2983, value=0.3165}
          <at 0.534 0.4866>
         ]}
        RedCurve {keys [<at 0.0485 4 RCIM_Cubic>, {arriveTangent=-7.6897, interpMode=RCIM_Cubic, leaveTangent=-7.6897, time=0.2983, value=2}, <at 0.534 0.2666>]}
       }}
      CurveIndex=$Emitter.Age
     ))
    ScaleA=false
    ScaleRGB=true
   ) {module=ScaleColor001}
  ]
  rendererClasses [&NiagaraSpriteRendererProperties]
  renderers [{rendererClass=&NiagaraSpriteRendererProperties}]
 }
 ElectricStreaks {
  emitterUpdateScript [
   &EmitterState(Life Cycle Mode=::NewEnumerator1, Loop Behavior=::NewEnumerator1, Loop Duration=0.17, UseLoopDelay=false)
   &SpawnBurst_Instantaneous(Spawn Count=8, Use Spawn Probability=false, UseLoopCountLimit=false)
   &SpawnBurst_Instantaneous(Spawn Count=3, Spawn Group=1, Spawn Time=0.05, Use Spawn Probability=false, UseLoopCountLimit=false) {module=SpawnBurst_Instantaneous001}
  ]
  particleSpawnScript [
   &InitializeParticle(
    Color=&RandomRangeLinearColor(Minimum=<rgba 1.9477 1.0986 4 1>, Maximum=<rgba 4.19 2.4994 6 1>)
    Lifetime=&UniformRangedFloat(Minimum=0.55, Maximum=0.65)
    Module.Write Color=true
    Module.Write Lifetime=true
    Module.Write Mass=true
    Module.Write Material Random=true
    Module.Write Position=false
    Module.Write Scale=false
    Module.Write SpriteRotation=false
    Module.Write SpriteSize=true
    Position=&SimulationPosition()
    Sprite Size=&UniformRangedVector2D(Minimum=<70 300>, Maximum=<90 370>)
   )
   &DynamicMaterialParameters(
    Index 0 Param 1=&RandomRangeFloat(Minimum=0, Maximum=1)
    Index 0 Param 2=&Lerp_Float(A=0.7, B=1.25, Alpha=$Particles.MaterialRandom)
    Module.Param0WriteEnabled=true
    Module.Param1WriteEnabled=true
    Module.Param2WriteEnabled=false
    Module.Param3WriteEnabled=false
   )
   &CameraOffset(Camera Offset Amount=25)
   &AddVelocity(
    Cone Angle=140
    Cone Axis=<0 0 1>
    Module.UseConeFalloff=true
    Module.UseVelDistribution=false
    Override Seed=false
    Speed Falloff From Cone Axis=-1
    Velocity Mode=::NewEnumerator2
    Velocity Speed=100
   ) {enabled=false}
   &SpriteFacingAndAlignment(
    Sprite Alignment=&MaskVectorBySpawnGroup(
     Masked Vector=&RotateVectorByQuaternion(
      Vector To Rotate=<0.5 0 0.5>
      Quaternion=&MakeQuatFromAxisAngle(Axis=<0 0 1>, AngleInDegrees=&RandomRangeFloat(Minimum=-50, Maximum=50), NormalizeAxis=true)
     )
     Passthrough Vector=&RotateVectorByQuaternion(
      Vector To Rotate=&RandomRangeVector(Minimum=<0.8 0 0.4>, Maximum=<1 0 0.1>)
      Quaternion=&MakeQuatFromAxisAngle(
       Axis=<0 0 1>
       AngleInDegrees=&Add_Float(
        A=&ReturnNormalizedExecIndex(Normalized Index Scale=360)
        B=&RandomRangeFloat(Minimum=-20, Maximum=20)
       )
       NormalizeAxis=true
      )
     )
     Spawn Group Mask=1
    )
    WriteAlignment=true
    WriteFacing=false
   )
  ]
  particleUpdateScript [
   &ParticleState()
   &SolveForcesAndVelocity(Clamp Velocity=false, Limit Acceleration=false)
   &ScaleSpriteSize(
    Non-Uniform Curve Sprite Scale=DataInterface() {propertyValues {
      XCurve {keys [
        {arriveTangent=0.8718, interpMode=RCIM_Cubic, leaveTangent=0.8718, tangentMode=RCTM_User, time=0, value=0.8}
        {arriveTangentWeight=0.1795, interpMode=RCIM_Cubic, leaveTangentWeight=0.1795, tangentMode=RCTM_User, time=0.5, value=1}
        {arriveTangent=-3.9093, interpMode=RCIM_Cubic, leaveTangent=-3.9093, tangentMode=RCTM_User, time=1, value=0.2}
       ]}
      YCurve {keys [
        {arriveTangent=2.2429, interpMode=RCIM_Cubic, leaveTangent=2.2429, tangentMode=RCTM_User, time=0, value=0.5}
        {arriveTangent=0.3913, arriveTangentWeight=0.1667, interpMode=RCIM_Cubic, leaveTangent=0.3913, tangentMode=RCTM_User, time=0.5, value=1}
        {arriveTangent=0.1824, interpMode=RCIM_Cubic, leaveTangent=0.1824, tangentMode=RCTM_User, time=1, value=1.12}
       ]}
     }}
    Scale Sprite Size Mode=::NewEnumerator3
   )
   &DynamicMaterialParameters(
    Index 0 Param 3=&FloatFromCurve(FloatCurve=DataInterface() {propertyValues {Curve {keys [<at 0.5 0 RCIM_Cubic>, {arriveTangent=3.8263, interpMode=RCIM_Cubic, leaveTangent=3.8263, tangentMode=RCTM_User, time=1, value=1}]}}})
    Index 0 Param 4=&Divide_Float(
     A=&CalculateParticleRadius(
      Radius Calculation Type:ENiagaraCollisionRadiusOptions=::NewEnumerator0
      Particle Radius:NiagaraFloat=1
      Mesh Dimensions:Vector3f=<1 1 1>
      Method for Calculating Particle Radius:ENiagaraCalculateRadiusOptions=::NewEnumerator2
     )
     B=200
    )
    Index 1 Param 2=$System.RippleMult
    Index 1 Param 3=$System.RipplePan
    Index 1 Param 4=600
    Module.Param0WriteEnabled=false
    Module.Param0WriteEnabled1=true
    Module.Param1WriteEnabled=false
    Module.Param1WriteEnabled1=true
    Module.Param2WriteEnabled=true
    Module.Param2WriteEnabled1=true
    Module.Param3WriteEnabled=true
    Module.Param3WriteEnabled1=true
    Write Parameter Index 1=true
   ) {module=DynamicMaterialParameters001}
   &ScaleColor(
    Scale Alpha=&FloatFromCurve(FloatCurve=DataInterface() {propertyValues {Curve {keys [<at 0.5 1 RCIM_Cubic>,<at 1 0>]}}})
    Scale RGB=&MakeVectorFromLinearColorRGB(Color=&ColorFromCurve(
      ColorCurve=DataInterface() {propertyValues {
        AlphaCurve {keys []}
        BlueCurve {keys [
          <at 0.0485 1.8072 RCIM_Cubic>
          {arriveTangent=-2.3921, interpMode=RCIM_Cubic, leaveTangent=-2.3921, time=0.2983, value=1.6918}
          <at 0.534 0.6458>
         ]}
        GreenCurve {keys [
          <at 0.0485 0.4164 RCIM_Cubic>
          {arriveTangent=0.1445, interpMode=RCIM_Cubic, leaveTangent=0.1445, time=0.2983, value=0.3165}
          <at 0.534 0.4866>
         ]}
        RedCurve {keys [<at 0.0485 4 RCIM_Cubic>, {arriveTangent=-7.6897, interpMode=RCIM_Cubic, leaveTangent=-7.6897, time=0.2983, value=2}, <at 0.534 0.2666>]}
       }}
      CurveIndex=$Emitter.Age
     ))
    ScaleA=true
    ScaleRGB=true
   )
  ]
  rendererClasses [&NiagaraSpriteRendererProperties]
  renderers [{rendererClass=&NiagaraSpriteRendererProperties}]
 }
 GroundWave {
  emitterUpdateScript [
   &EmitterState()
   &SpawnBurst_Instantaneous(Use Spawn Probability=false, UseLoopCountLimit=false)
   &SpawnBurst_Instantaneous(Spawn Group=1, Spawn Time=0.1, Use Spawn Probability=false, UseLoopCountLimit=false) {module=SpawnBurst_Instantaneous001}
  ]
  particleSpawnScript [
   /Niagara/Modules/Spawn/Initialization/V2/InitializeParticle(
    Color=<rgba 1.25 1.25 1.25 1>
    Lifetime=&MaskFloatBySpawnGroup(Masked Float=0.25, Passthrough Float=0.5)
    Mesh Scale Mode=::NewEnumerator3
    Mesh Uniform Scale=&MaskFloatBySpawnGroup(Masked Float=2.5, Passthrough Float=3.5)
    UsePositionOffset=false
    Write Material Random=true
   )
   &InitialMeshOrientation(
    Mesh Orientation Mode=::NewEnumerator2
    Rotation=&RandomRangeVector(Minimum=<0 0 0>, Maximum=<0 0 1>)
    Use Rotation Vector=true
   )
   &CameraOffset(Camera Offset Amount=40)
  ]
  particleUpdateScript [
   &ParticleState()
   &ScaleMeshSize(Scale Factor=&VectorFromCurve(VectorCurve=DataInterface() {propertyValues {
       XCurve {keys [
         {arriveTangent=1.4912, interpMode=RCIM_Cubic, leaveTangent=1.4912, tangentMode=RCTM_User, time=0, value=0.2}
         {arriveTangent=0.1443, interpMode=RCIM_Cubic, leaveTangent=0.1443, tangentMode=RCTM_User, time=1, value=1}
        ]}
       YCurve {keys [
         {arriveTangent=1.4912, interpMode=RCIM_Cubic, leaveTangent=1.4912, tangentMode=RCTM_User, time=0, value=0.2}
         {arriveTangent=0.1443, interpMode=RCIM_Cubic, leaveTangent=0.1443, tangentMode=RCTM_User, time=1, value=1}
        ]}
       ZCurve {keys [
         {arriveTangent=5.9564, interpMode=RCIM_Cubic, leaveTangent=5.9564, tangentMode=RCTM_User, time=0, value=0.5}
         {arriveTangent=-0.1, interpMode=RCIM_Cubic, leaveTangent=-0.1, time=0.2, value=1}
         {arriveTangent=-1.1924, interpMode=RCIM_Cubic, leaveTangent=-1.1924, tangentMode=RCTM_User, time=1, value=0.4}
        ]}
      }})) {enabled=false}
   &ScaleMeshSize(Scale Factor=&MaskVectorBySpawnGroup(
     Masked Vector=&VectorFromCurve(VectorCurve=DataInterface() {propertyValues {
        XCurve {keys [
          {arriveTangent=2.2781, interpMode=RCIM_Cubic, leaveTangent=2.2781, tangentMode=RCTM_User, time=0, value=0.1}
          {arriveTangent=0.288, interpMode=RCIM_Cubic, leaveTangent=0.288, tangentMode=RCTM_User, time=1, value=1}
         ]}
        YCurve {keys [
          {arriveTangent=2.2781, interpMode=RCIM_Cubic, leaveTangent=2.2781, tangentMode=RCTM_User, time=0, value=0.1}
          {arriveTangent=0.288, interpMode=RCIM_Cubic, leaveTangent=0.288, tangentMode=RCTM_User, time=1, value=1}
         ]}
        ZCurve {keys [
          {arriveTangent=3.6292, interpMode=RCIM_Cubic, leaveTangent=3.6292, tangentMode=RCTM_User, time=0, value=0.5}
          {arriveTangent=-1.1338, interpMode=RCIM_Cubic, leaveTangent=-1.1338, tangentMode=RCTM_User, time=0.3, value=0.6}
          {arriveTangent=-0.1814, interpMode=RCIM_Cubic, leaveTangent=-0.1814, tangentMode=RCTM_User, time=1, value=0.4}
         ]}
       }})
     Passthrough Vector=&VectorFromCurve(VectorCurve=DataInterface() {propertyValues {
        XCurve {keys [
          {arriveTangent=3.9002, interpMode=RCIM_Cubic, leaveTangent=3.9002, tangentMode=RCTM_User, time=0, value=0.2}
          {arriveTangent=1.2534, interpMode=RCIM_Cubic, leaveTangent=1.2534, tangentMode=RCTM_User, time=0.1743, value=0.6343}
          {arriveTangent=0.2006, interpMode=RCIM_Cubic, leaveTangent=0.2006, tangentMode=RCTM_User, time=1, value=1}
         ]}
        YCurve {keys [
          {arriveTangent=3.9002, interpMode=RCIM_Cubic, leaveTangent=3.9002, tangentMode=RCTM_User, time=0, value=0.2}
          {arriveTangent=1.2534, interpMode=RCIM_Cubic, leaveTangent=1.2534, tangentMode=RCTM_User, time=0.1743, value=0.6343}
          {arriveTangent=0.2006, interpMode=RCIM_Cubic, leaveTangent=0.2006, tangentMode=RCTM_User, time=1, value=1}
         ]}
        ZCurve {keys [
          {arriveTangent=8.292, interpMode=RCIM_Cubic, leaveTangent=8.292, tangentMode=RCTM_User, time=0, value=0.16}
          {arriveTangent=-1.3261, interpMode=RCIM_Cubic, leaveTangent=-1.3261, tangentMode=RCTM_User, time=0.07, value=0.6}
          <at 1 0.2>
         ]}
       }})
    )) {module=ScaleMeshSize001}
   &DynamicMaterialParameters(
    Index 0 Param 1=&MaskFloatBySpawnGroup(
     Masked Float=1
     Passthrough Float=&FloatFromCurve(FloatCurve=DataInterface() {propertyValues {Curve {keys [{arriveTangent=-3.0733, interpMode=RCIM_Cubic, leaveTangent=-3.0733, tangentMode=RCTM_User, time=0, value=1}, <at 1 -0.75 RCIM_Cubic>]}}})
    )
    Index 0 Param 2=&MaskFloatBySpawnGroup(
     Masked Float=1.75
     Passthrough Float=&FloatFromCurve(FloatCurve=DataInterface() {propertyValues {Curve {keys [
          {arriveTangent=-2.0952, interpMode=RCIM_Cubic, leaveTangent=-2.0952, tangentMode=RCTM_User, time=0, value=1.5}
          {arriveTangent=-1.4744, interpMode=RCIM_Cubic, leaveTangent=-1.4744, tangentMode=RCTM_User, time=1, value=0.25}
         ]}}})
    )
    Index 0 Param 3=&FloatFromCurve(
     FloatCurve=DataInterface() {propertyValues {Curve {keys [
         {arriveTangent=-2.0927, interpMode=RCIM_Cubic, leaveTangent=-2.0927, tangentMode=RCTM_User, time=0, value=1}
         {interpMode=RCIM_Cubic, leaveTangent=-1.7693, time=1, value=0}
        ]}}}
     Scale Curve=0.25
    )
    Index 0 Param 4=&FloatFromCurve(FloatCurve=DataInterface() {propertyValues {Curve {keys [{arriveTangent=1.0328, interpMode=RCIM_Cubic, leaveTangent=1.0328, tangentMode=RCTM_User, time=0, value=0}, <at 1 1 RCIM_Cubic>]}}})
    Module.Param0WriteEnabled=true
    Module.Param1WriteEnabled=true
    Module.Param2WriteEnabled=true
    Module.Param3WriteEnabled=true
   )
   &ScaleColor(
    Scale Alpha=&FloatFromCurve(FloatCurve=DataInterface() {propertyValues {Curve {keys [
         {arriveTangent=26.1682, interpMode=RCIM_Cubic, leaveTangent=26.1682, tangentMode=RCTM_User, time=0, value=0}
         {interpMode=RCIM_Cubic, tangentMode=RCTM_User, time=0.1, value=1}
         {tangentMode=RCTM_User, time=1, value=0}
        ]}}})
    Scale RGB=&Lerp_Vector(
     A=<3 1 1.5>
     B=<0.5 1 1.8>
     Alpha=&FloatFromCurve(
      FloatCurve=DataInterface() {propertyValues {Curve {keys [{interpMode=RCIM_Cubic, tangentMode=RCTM_User, time=0.1, value=0}, {interpMode=RCIM_Cubic, tangentMode=RCTM_User, time=0.4, value=1}]}}}
      CurveIndex=$Emitter.Age
     )
    )
    ScaleA=true
    ScaleRGB=true
   )
  ]
  rendererClasses [&NiagaraSpriteRendererProperties, &NiagaraMeshRendererProperties]
  renderers [
   {rendererClass=&NiagaraSpriteRendererProperties}
   {rendererClass=&NiagaraMeshRendererProperties, rendererIndex=1}
  ]
 }
 OuterSphere {
  emitterUpdateScript [
   &EmitterState()
   &SpawnBurst_Instantaneous(Use Spawn Probability=false, UseLoopCountLimit=false)
  ]
  particleSpawnScript [
   &InitializeParticle(
    Color=<rgba 1.5 1.5 1.5 1>
    Lifetime=0.45
    Mesh Scale=<3.2 3.2 3.2>
    Mesh Scale Mode=::NewEnumerator4
    UsePositionOffset=false
    Write Material Random=true
   )
   /SetVariables(Particles.LightPosition:NiagaraPosition=<0 0 40>) {isSetParameters=true}
  ]
  particleUpdateScript [
   &ParticleState()
   /SetVariables(Particles.Position=&MakeVector(
     X=0
     Y=0
     Z=&FloatFromCurve(
      FloatCurve=DataInterface() {propertyValues {Curve {keys [
          {arriveTangent=5.5536, interpMode=RCIM_Cubic, leaveTangent=5.5536, tangentMode=RCTM_User, time=0, value=0}
          {arriveTangent=1, interpMode=RCIM_Cubic, leaveTangent=1, time=0.2, value=0.7}
          {arriveTangent=0.1617, interpMode=RCIM_Cubic, leaveTangent=0.1617, tangentMode=RCTM_User, time=1, value=1}
         ]}}}
      Scale Curve=25
     )
    )) {isSetParameters=true}
   /SetVariables(Particles.LightColor=&ColorFromCurve(ColorCurve=DataInterface() {propertyValues {
       AlphaCurve {keys [
         <at 0 1 RCIM_Cubic>
         {arriveTangent=-2.4738, interpMode=RCIM_Cubic, leaveTangent=-2.4738, tangentMode=RCTM_User, time=0.3627, value=0.7008}
         <at 1 0 RCIM_Cubic>
        ]}
       BlueCurve {keys [<at 0.1733 36.9992 RCIM_Cubic>,<at 0.647 25 RCIM_Cubic>]}
       GreenCurve {keys [<at 0.1733 1.9736 RCIM_Cubic>,<at 0.647 10.9288 RCIM_Cubic>]}
       RedCurve {keys [<at 0.1733 40 RCIM_Cubic>,<at 0.647 4.126 RCIM_Cubic>]}
      }})) {isSetParameters=true}
   &ScaleMeshSize(Scale Factor=&VectorFromFloat(Value=&FloatFromCurve(FloatCurve=DataInterface() {propertyValues {Curve {keys [
          {arriveTangent=5.5536, interpMode=RCIM_Cubic, leaveTangent=5.5536, tangentMode=RCTM_User, time=0, value=0.2}
          {arriveTangent=0.8, interpMode=RCIM_Cubic, leaveTangent=0.8, time=0.2, value=0.8}
          {arriveTangent=0.1617, interpMode=RCIM_Cubic, leaveTangent=0.1617, tangentMode=RCTM_User, time=1, value=1}
         ]}}})))
   &DynamicMaterialParameters(
    Index 0 Param 1=&FloatFromCurve(
     FloatCurve=DataInterface() {propertyValues {Curve {keys [
         {arriveTangent=-2.9934, interpMode=RCIM_Cubic, leaveTangent=-2.9934, tangentMode=RCTM_User, time=0, value=1}
         {interpMode=RCIM_Cubic, leaveTangent=-1.2287, time=1, value=-0.25}
        ]}}}
     Scale Curve=2
    )
    Index 0 Param 2=1
    Index 0 Param 3=&FloatFromCurve(
     FloatCurve=DataInterface() {propertyValues {Curve {keys [
         {arriveTangent=2.2611, interpMode=RCIM_Cubic, leaveTangent=2.2611, tangentMode=RCTM_User, time=0, value=0}
         {arriveTangent=0.402, interpMode=RCIM_Cubic, leaveTangent=0.402, tangentMode=RCTM_User, time=1, value=1}
        ]}}}
     Scale Curve=5
    )
    Index 0 Param 4=&FloatFromCurve(FloatCurve=DataInterface() {propertyValues {Curve {keys [{arriveTangent=-2.6143, interpMode=RCIM_Cubic, time=0, value=1}, <at 0.7 0>]}}})
    Module.Param0WriteEnabled=true
    Module.Param1WriteEnabled=true
    Module.Param2WriteEnabled=true
    Module.Param3WriteEnabled=true
   )
   &ScaleColor(
    Scale Alpha=&FloatFromCurve(FloatCurve=DataInterface() {propertyValues {Curve {keys [
         {arriveTangent=26.1682, interpMode=RCIM_Cubic, leaveTangent=26.1682, tangentMode=RCTM_User, time=0, value=0}
         {arriveTangent=-2.2763, interpMode=RCIM_Cubic, leaveTangent=-2.2763, tangentMode=RCTM_User, time=0.07, value=1}
         {tangentMode=RCTM_User, time=1, value=0}
        ]}}})
    Scale RGB=&Lerp_Vector(
     A=<1.2 0.2 0.8>
     B=<0.2 1 1>
     Alpha=&FloatFromCurve(
      FloatCurve=DataInterface() {propertyValues {Curve {keys [{interpMode=RCIM_Cubic, tangentMode=RCTM_User, time=0.1, value=0}, {interpMode=RCIM_Cubic, tangentMode=RCTM_User, time=0.4, value=1}]}}}
      CurveIndex=$Emitter.Age
     )
    )
    ScaleA=true
    ScaleRGB=true
   )
  ]
  rendererClasses [&NiagaraMeshRendererProperties, &NiagaraLightRendererProperties]
  renderers [
   {rendererClass=&NiagaraMeshRendererProperties, rendererIndex=0}
   {rendererClass=&NiagaraLightRendererProperties, rendererIndex=1}
  ]
 }
 Rocks {
  emitterUpdateScript [
   &EmitterState(Enable Distance Culling=true, MaxDistance=3000, Module.bUseMaxDistance=true, Module.bUseMinDistance=false, Scalability Mode=::NewEnumerator1)
   &SpawnBurst_Instantaneous(Spawn Count=8, Spawn Time=0.05, Use Spawn Probability=false, UseLoopCountLimit=false)
  ]
  particleSpawnScript [
   &InitializeParticle(
    Color Maximum=<rgba 0.5 0.5 0.5 1>
    Color Minimum=<rgba 0.25 0.25 0.25 1>
    Color Mode=::NewEnumerator2
    Lifetime=1.5
    Mesh Scale Mode=::NewEnumerator1
    Mesh Uniform Scale Max=0.32
    Mesh Uniform Scale Min=0.12
    UsePositionOffset=false
    Write Material Random=true
   )
   &InitialMeshOrientation(
    Rotation=&RandomRangeVector(Minimum=<0 0 0>, Maximum=<1 1 1>)
    Use Orientation Vector=true
    Use Rotation Vector=true
   )
   &ShapeLocation(
    MaskSpawnGroup=false
    Non Uniform Scale=<1 1 0.25>
    Offset=<0 0 15>
    Offset Mode=::NewEnumerator0
    Override Seed=false
    Sphere Radius=25
    UseOwnerScale=true
   )
   &AddVelocity(
    Module.Use Falloff Distance=false
    Origin Offset=<0 0 -7>
    Override Seed=false
    Velocity Mode=::NewEnumerator1
    Velocity Speed=&RandomRangeFloat(Minimum=700, Maximum=1400)
   )
   /SetVariables(Particles.SpriteSize=&Vector2DFromFloat(Value=&Multiply_Float(A=&MakeFloatFromVector(Vector=$Particles.Scale, Channel=::NewEnumerator0), B=150))) {isSetParameters=true}
   &DynamicMaterialParameters(
    Index 3 Param 1=4
    Index 3 Param 2=0.2
    Module.Param0WriteEnabled=false
    Module.Param0WriteEnabled3=true
    Module.Param1WriteEnabled=false
    Module.Param1WriteEnabled3=true
    Module.Param2WriteEnabled=false
    Module.Param2WriteEnabled3=true
    Module.Param3WriteEnabled=false
    Module.Param3WriteEnabled3=true
    Write Parameter Index 3=true
   )
  ]
  particleUpdateScript [
   &ParticleState()
   &ScaleSpriteSize(Uniform Curve Sprite Scale=DataInterface() {propertyValues {Curve {keys [
        <at 0 1 RCIM_Cubic>
        {arriveTangent=-0.5, interpMode=RCIM_Cubic, leaveTangent=-0.5, time=0.1, value=1.25}
        {arriveTangent=-3.7303, interpMode=RCIM_Cubic, leaveTangent=-3.7303, tangentMode=RCTM_User, time=0.2, value=0.9}
        {interpMode=RCIM_Cubic, tangentMode=RCTM_User, time=1, value=0}
       ]}}})
   &GravityForce(Gravity=<0 0 -1600>)
   &DoOnce(Trigger Condition:NiagaraBool=&Boolean_AndOperation(
     Bool1=&SetBoolByFloatComparison(A=&VectorLength(Vector=$Particles.Velocity), B=150, Comparison Type=::NewEnumerator6)
     Bool2=&SetBoolByFloatComparison(A=$Particles.Age, B=0.5)
    ))
   &Collision(
    Advanced Aging Rate=1
    Age Colliding Particles=false
    CPU Collision Trace Channel=ECC_Visibility
    Collision Enabled=&Boolean_NotOperation(Bool=$Particles.DoOnce.TriggeredOnce)
    Friction=0.5
    Friction During a Bounce=0.3
    Module.Control Rest State Via Surface Normal=false
    Module.EnableMaxCollisionCount=false
    Module.Randomize Collision Normal Vector=true
    Particle Radius Scale=2
    Radius Calculation Type=::NewEnumerator1
    Randomize Collision Normal=0.25
    Restitution=0.4
   )
   &MeshRotationForce(Rotation=&RandomRangeVector(Minimum=<-10 -10 -10>, Maximum=<10 10 10>))
   &SolveForcesAndVelocity(Clamp Velocity=false, Limit Acceleration=false)
   &ScaleMeshSize(Scale Factor=&VectorFromFloat(Value=&FloatFromCurve(FloatCurve=DataInterface() {propertyValues {Curve {keys [
          {arriveTangentWeight=0.2357, interpMode=RCIM_Cubic, leaveTangentWeight=0.2357, tangentMode=RCTM_User, time=0.5, value=1}
          {arriveTangent=-4.4912, interpMode=RCIM_Cubic, leaveTangent=-4.4912, tangentMode=RCTM_User, time=1, value=0}
         ]}}})))
   /SetVariables(Particles.SpriteColor=&MakeLinearColorFromVectorAndFloat(
     "Vector (RGB)"=&MakeVectorFromLinearColorRGB(Color=<rgba 2.636 0.377 3 0.5>)
     "Float (Alpha)"=&FloatFromCurve(FloatCurve=DataInterface() {propertyValues {Curve {keys [
          <at 0 1 RCIM_Cubic>
          {arriveTangent=-2.8631, interpMode=RCIM_Cubic, leaveTangent=-2.8631, tangentMode=RCTM_User, time=0.1155, value=0.9}
          <at 0.2748 0.3227>
          <at 1 0>
         ]}}})
    )) {isSetParameters=true}
  ]
  rendererClasses [&NiagaraMeshRendererProperties, &NiagaraSpriteRendererProperties]
  renderers [
   {rendererClass=&NiagaraMeshRendererProperties, rendererIndex=0}
   {rendererClass=&NiagaraSpriteRendererProperties, rendererIndex=1}
  ]
 }
 Wind {
  emitterUpdateScript [
   &EmitterState(Enable Distance Culling=true, MaxDistance=6000, Module.bUseMaxDistance=true, Module.bUseMinDistance=false, Scalability Mode=::NewEnumerator1)
   &SpawnRate(
    SpawnRate=&FloatFromCurve(
     FloatCurve=DataInterface() {propertyValues {Curve {keys [
         {arriveTangent=77.1005, interpMode=RCIM_Cubic, leaveTangent=77.1005, tangentMode=RCTM_User, time=0.05, value=0}
         <at 0.07 1 RCIM_Cubic>
         {arriveTangent=-14.3902, interpMode=RCIM_Cubic, leaveTangent=-14.3902, tangentMode=RCTM_User, time=0.25, value=0}
        ]}}}
     CurveIndex=$System.Age
     Scale Curve=40
    )
    Use Spawn Probability=false
   )
   &SpawnBurst_Instantaneous(Spawn Count=4, Spawn Time=0.05, Use Spawn Probability=false, UseLoopCountLimit=false)
  ]
  particleSpawnScript [
   &InitializeParticle(
    Color=<rgba 1 1 1 0.35>
    Lifetime=&UniformRangedFloat(Minimum=0.3, Maximum=0.47)
    Mesh Scale=&Lerp_Vector(A=<5 5 2.5>, B=<6.5 6.5 3.5>, Alpha=&RandomRangeFloat(Minimum=0, Maximum=1))
    Mesh Scale Mode=::NewEnumerator4
    UsePositionOffset=false
    Write Material Random=true
   ) {module=InitializeParticle001}
   &InitialMeshOrientation(
    Mesh Axis To Orient=<0 0 1>
    Mesh Orientation Mode=::NewEnumerator2
    Orientation Vector=&RandomRangeVector(Minimum=<0 0 1>, Maximum=<0 0 1>)
    Rotation=&RandomRangeVector(Minimum=<0 0 0>, Maximum=<0 0 1>)
    Use Rotation Vector=true
   )
  ]
  particleUpdateScript [
   &ParticleState()
   &ScaleMeshSize(Scale Factor=&VectorFromCurve(VectorCurve=DataInterface() {propertyValues {
       XCurve {keys [
         {arriveTangent=1.9177, interpMode=RCIM_Cubic, leaveTangent=1.9177, tangentMode=RCTM_User, time=0, value=0.1}
         {arriveTangent=0.2969, interpMode=RCIM_Cubic, leaveTangent=0.8713, tangentMode=RCTM_User, time=1, value=1}
        ]}
       YCurve {keys [
         {arriveTangent=1.9177, interpMode=RCIM_Cubic, leaveTangent=1.9177, tangentMode=RCTM_User, time=0, value=0.1}
         {arriveTangent=0.2969, interpMode=RCIM_Cubic, leaveTangent=0.5259, tangentMode=RCTM_User, time=1, value=1}
        ]}
       ZCurve {keys [
         {arriveTangent=2.8674, interpMode=RCIM_Cubic, leaveTangent=2.8674, tangentMode=RCTM_User, time=0, value=0.5}
         {arriveTangent=-0.4, interpMode=RCIM_Cubic, leaveTangent=-0.4, time=0.35, value=1}
         {arriveTangent=-1.6862, interpMode=RCIM_Cubic, leaveTangent=-1.6862, tangentMode=RCTM_User, time=1, value=0.1}
        ]}
      }}))
   /SetVariables(Particles.Position=&AddVectorToPosition(
     Position=&Multiply_VectorByFloat(
      Vector=&ReturnMeshOrientationAxis(Mesh Orientation Axis=::XAxis, Mesh Orientation Quat=$Particles.MeshOrientation)
      Float=&FloatFromCurve(
       FloatCurve=DataInterface() {propertyValues {Curve {keys [
           {arriveTangent=3.4662, interpMode=RCIM_Cubic, leaveTangent=3.4662, tangentMode=RCTM_User, time=0, value=-0.2}
           {interpMode=RCIM_Cubic, tangentMode=RCTM_User, time=0.5, value=1}
          ]}}}
       CurveIndex=$Emitter.Age
       Scale Curve=&Multiply_Float(A=&MakeFloatFromVector(Vector=$Particles.Initial.Scale, Channel=::NewEnumerator0), B=18)
      )
     )
     Vector=&MakeVector(X=0, Y=0, Z=&RandomRangeFloat(Minimum=25, Maximum=60))
    )) {isSetParameters=true}
   &CameraOffset(Camera Offset Amount=&Lerp_Float(A=70, B=140, Alpha=$Particles.NormalizedAge))
   &ScaleColor(
    Scale Alpha=&FloatFromCurve(FloatCurve=DataInterface() {propertyValues {Curve {keys [
         {arriveTangent=8.125, interpMode=RCIM_Cubic, time=0, value=0}
         {interpMode=RCIM_Cubic, tangentMode=RCTM_User, time=0.3, value=1}
         {arriveTangent=-0.8126, interpMode=RCIM_Cubic, leaveTangent=-0.8126, tangentMode=RCTM_User, time=0.7791, value=0.7}
         {interpMode=RCIM_Cubic, leaveTangent=-2.3395, time=1, value=0}
        ]}}})
    Scale RGB=&MakeVectorFromLinearColorRGB(Color=&ColorFromCurve(
      ColorCurve=DataInterface() {propertyValues {
        AlphaCurve {keys []}
        BlueCurve {keys [<at 0.0485 6 RCIM_Cubic>, {arriveTangent=-19.0486, interpMode=RCIM_Cubic, leaveTangent=-19.0486, time=0.1696, value=2}, <at 0.3295 0.6458>]}
        GreenCurve {keys [
          <at 0.0485 1.018 RCIM_Cubic>
          {arriveTangent=-2.2041, interpMode=RCIM_Cubic, leaveTangent=-2.2041, time=0.1696, value=0.4856}
          <at 0.3295 0.3985>
         ]}
        RedCurve {keys [
          <at 0.0485 5.153 RCIM_Cubic>
          {arriveTangent=-16.8034, interpMode=RCIM_Cubic, leaveTangent=-16.8034, time=0.1696, value=1.4424}
          <at 0.3295 0.4299>
         ]}
       }}
      CurveIndex=$Emitter.Age
     ))
    ScaleA=true
    ScaleRGB=true
   )
   &DynamicMaterialParameters(
    Index 0 Param 1=&FloatFromCurve(FloatCurve=DataInterface() {propertyValues {Curve {keys [
         {arriveTangent=-13.4039, interpMode=RCIM_Cubic, leaveTangent=-13.4039, tangentMode=RCTM_User, time=0, value=1}
         <at 0.1 0>
         {arriveTangent=1.1111, interpMode=RCIM_Cubic, leaveTangent=1.1111, time=0.4, value=0}
         {interpMode=RCIM_Cubic, leaveTangent=16.4602, time=1, value=1}
        ]}}})
    Index 0 Param 2=&FloatFromCurve(
     FloatCurve=DataInterface() {propertyValues {Curve {keys [
         {arriveTangent=2.4843, interpMode=RCIM_Cubic, leaveTangent=2.4843, tangentMode=RCTM_User, time=0, value=0}
         {interpMode=RCIM_Cubic, leaveTangent=-0.8461, time=1, value=1}
        ]}}}
     Scale Curve=0.5
    )
    Index 0 Param 4=&FloatFromCurve(
     FloatCurve=DataInterface() {propertyValues {Curve {keys [
         {arriveTangent=0.2805, interpMode=RCIM_Cubic, time=0, value=1}
         {arriveTangent=2.4494, interpMode=RCIM_Cubic, leaveTangent=2.4494, tangentMode=RCTM_User, time=1, value=2}
        ]}}}
     Scale Curve=&RandomRangeFloat(Minimum=0.6, Maximum=0.8)
    )
    Module.Param0WriteEnabled=true
    Module.Param1WriteEnabled=true
    Module.Param2WriteEnabled=false
    Module.Param3WriteEnabled=true
   )
  ]
  rendererClasses [&NiagaraMeshRendererProperties]
  renderers [{rendererClass=&NiagaraMeshRendererProperties, rendererIndex=0}]
 }
 WindCone {
  emitterUpdateScript [
   &EmitterState()
   &SpawnBurst_Instantaneous(Use Spawn Probability=false, UseLoopCountLimit=false)
  ]
  particleSpawnScript [
   &InitializeParticle(
    Color=<rgba 8.5464 0.1729 10 0.5>
    Lifetime=0.3
    Mesh Scale Mode=::NewEnumerator4
    Sprite Size Mode=::NewEnumerator3
    Uniform Sprite Size=250
    UsePositionOffset=false
    Write Material Random=true
   ) {module=InitializeParticle001}
   &DynamicMaterialParameters(
    Index 0 Param 1=0.7
    Index 0 Param 2=0.8
    Index 0 Param 3=2
    Index 2 Param 1=10
    Index 3 Param 1=0.2
    Index 3 Param 2=3
    Index 3 Param 3=1
    Index 3 Param 4=1
    Module.Param0WriteEnabled=true
    Module.Param0WriteEnabled2=true
    Module.Param0WriteEnabled3=true
    Module.Param1WriteEnabled=true
    Module.Param1WriteEnabled2=false
    Module.Param1WriteEnabled3=true
    Module.Param2WriteEnabled=true
    Module.Param2WriteEnabled2=false
    Module.Param2WriteEnabled3=true
    Module.Param3WriteEnabled=false
    Module.Param3WriteEnabled2=false
    Module.Param3WriteEnabled3=true
    Write Parameter Index 2=true
    Write Parameter Index 3=true
   )
   &InitialMeshOrientation(Mesh Orientation Mode=::NewEnumerator2, Orientation Vector=<0 0 -1>, Use Rotation Vector=false)
   &CameraOffset(Camera Offset Amount=35)
  ]
  particleUpdateScript [
   &ParticleState()
   &ScaleMeshSize(Scale Factor=&VectorFromCurve(VectorCurve=DataInterface() {propertyValues {
       XCurve {keys [
         {arriveTangent=1.5385, time=0, value=0}
         {arriveTangent=0.3483, interpMode=RCIM_Cubic, leaveTangent=0.3483, tangentMode=RCTM_User, time=1, value=1}
        ]}
       YCurve {keys [
         {arriveTangent=2.3109, interpMode=RCIM_Cubic, leaveTangent=2.3109, tangentMode=RCTM_User, time=0, value=0}
         {arriveTangent=0.6124, interpMode=RCIM_Cubic, leaveTangent=0.6124, tangentMode=RCTM_User, time=1, value=1}
        ]}
       ZCurve {keys [
         {arriveTangent=2.3109, interpMode=RCIM_Cubic, leaveTangent=2.3109, tangentMode=RCTM_User, time=0, value=0}
         {arriveTangent=0.6124, interpMode=RCIM_Cubic, leaveTangent=0.6124, tangentMode=RCTM_User, time=1, value=1}
        ]}
      }}))
   &ScaleColor(
    Scale Alpha=&FloatFromCurve(FloatCurve=DataInterface() {propertyValues {Curve {keys [
         {arriveTangent=10809.126, interpMode=RCIM_Cubic, time=0, value=0}
         <at 0.15 1 RCIM_Cubic>
         {interpMode=RCIM_Cubic, leaveTangent=-3.1778, time=1, value=0}
        ]}}})
    Scale RGB=&MakeVectorFromLinearColorRGB(Color=&ColorFromCurve(ColorCurve=DataInterface() {propertyValues {
        AlphaCurve {keys []}
        BlueCurve {keys [<at 0.1026 1 RCIM_Cubic>,<at 0.959 0.5 RCIM_Cubic>]}
        GreenCurve {keys [<at 0.1026 1 RCIM_Cubic>,<at 0.959 0.2461 RCIM_Cubic>]}
        RedCurve {keys [<at 0.1026 1 RCIM_Cubic>,<at 0.959 0.2651 RCIM_Cubic>]}
       }}))
    ScaleA=true
    ScaleRGB=true
   )
   /SetVariables(Particles.FlareColor=&MultiplyLinearColor(Linear Color=$Particles.Color, Scale Factor=<2 2 2 3>)) {isSetParameters=true}
   &DynamicMaterialParameters(
    Index 1 Param 1=&FloatFromCurve(FloatCurve=DataInterface() {propertyValues {Curve {keys [<at 0 -1 RCIM_Cubic>,<at 1 0>]}}})
    Module.Param0WriteEnabled1=true
    Module.Param1WriteEnabled1=true
    Module.Param2WriteEnabled1=false
    Module.Param3WriteEnabled1=false
    Write Parameter Index 0=false
    Write Parameter Index 1=true
   ) {module=DynamicMaterialParameters001}
  ]
  rendererClasses [&NiagaraMeshRendererProperties, &NiagaraSpriteRendererProperties]
  renderers [
   {rendererClass=&NiagaraMeshRendererProperties, rendererIndex=0}
   {rendererClass=&NiagaraSpriteRendererProperties, rendererIndex=1}
  ]
 }
 debrisGPU_high {
  simTarget=GPUComputeSim
  emitterUpdateScript [
   &EmitterState(Enable Distance Culling=true, MaxDistance=3500, Module.bUseMaxDistance=true, Module.bUseMinDistance=false, Scalability Mode=::NewEnumerator1)
   &SpawnBurst_Instantaneous(Spawn Count=15, Spawn Time=0.02, Use Spawn Probability=false, UseLoopCountLimit=false)
   &SpawnRate(
    Spawn Probability=0.75
    SpawnRate=&Multiply_Float(A=&RandomRangeFloat(Minimum=25, Maximum=50), B=$User.Magnitude)
    Use Spawn Probability=true
   )
  ]
  particleSpawnScript [
   &InitializeParticle(
    Color Maximum=<rgba 0.5 0.5 0.5 1>
    Color Minimum=<rgba 0.2 0.1823 0.1762 1>
    Color Mode=::NewEnumerator2
    Lifetime Max=1.25
    Lifetime Min=0.75
    Lifetime Mode=::NewEnumerator1
    Sprite Rotation Mode=::NewEnumerator2
    Sprite Size Max=<11 11>
    Sprite Size Mode=::NewEnumerator2
    UsePositionOffset=false
    Write Material Random=true
   ) {module=InitializeParticle001}
   &ShapeLocation(
    MaskSpawnGroup=false
    Non Uniform Scale=<1 1 0.25>
    Offset=<0 0 15>
    Offset Mode=::NewEnumerator0
    Override Seed=false
    Sphere Radius=30
    UseOwnerScale=true
   )
   &AddVelocity(
    Module.Use Falloff Distance=false
    Origin Offset=<0 0 -15>
    Override Seed=false
    Velocity Mode=::NewEnumerator1
    Velocity Speed=&RandomRangeFloat(Minimum=500, Maximum=1500)
   )
   &SubUVAnimation(End Frame=3, SubUV Animation Mode:ENiagara_SubUVLookupMode=::NewEnumerator2) {module=SubUVAnimation001}
  ]
  particleUpdateScript [
   &ParticleState()
   &GravityForce(Gravity=<0 0 -1600>)
   &SpriteRotationRate(Rotation Rate=&MakeCustomFloatFromBool(Bool=&RandomBool(), True Float=-1080, False Float=1080))
   &SolveForcesAndVelocity(Clamp Velocity=false, Limit Acceleration=false)
   &ScaleSpriteSize(Uniform Curve Sprite Scale=DataInterface() {propertyValues {Curve {keys [
        {arriveTangent=11.5942, interpMode=RCIM_Cubic, leaveTangent=11.5942, tangentMode=RCTM_User, time=0, value=0}
        <at 0.15 1>
        {arriveTangentWeight=0.2357, interpMode=RCIM_Cubic, leaveTangentWeight=0.2357, tangentMode=RCTM_User, time=0.5, value=1}
        {arriveTangent=-4.4912, interpMode=RCIM_Cubic, leaveTangent=-4.4912, tangentMode=RCTM_User, time=1, value=0}
       ]}}})
  ]
  rendererClasses [&NiagaraSpriteRendererProperties]
  renderers [{rendererClass=&NiagaraSpriteRendererProperties}]
 }
}
summary {
 systemName=NS_ShockwaveMace_Bounce
 emitters [
  {emitterName=GroundWave, rendererClasses [&NiagaraSpriteRendererProperties, &NiagaraMeshRendererProperties]}
  {emitterName=OuterSphere, rendererClasses [&NiagaraMeshRendererProperties, &NiagaraLightRendererProperties]}
  {emitterName=WindCone, rendererClasses [&NiagaraMeshRendererProperties, &NiagaraSpriteRendererProperties]}
  {emitterName=Dust, rendererClasses [&NiagaraSpriteRendererProperties]}
  {emitterName=Wind, rendererClasses [&NiagaraMeshRendererProperties]}
  {emitterName=ElectricArcs, rendererClasses [&NiagaraSpriteRendererProperties]}
  {emitterName=ElectricStreaks, rendererClasses [&NiagaraSpriteRendererProperties]}
  {emitterName=debrisGPU_high, simTarget=GPUComputeSim, rendererClasses [&NiagaraSpriteRendererProperties]}
  {emitterName=Rocks, rendererClasses [&NiagaraMeshRendererProperties, &NiagaraSpriteRendererProperties]}
 ]
 userVariables [User.Magnitude {type=NiagaraFloat, defaultValue=0, description=""}]
}
