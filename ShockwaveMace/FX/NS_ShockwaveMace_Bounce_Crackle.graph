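# Niagara crackle effect for the Shockwave Mace bounce, placed at an offset from a skeletal mesh supplied as a user parameter.
# Short-lived magenta electric arc sprites and a pulsing glow sprite with a light resample the mesh location every update to follow it.
# OuterSphere and Wind mesh particles sample it only at spawn, then animate scale, color and dynamic material parameters, with shells shifting pink to cyan and wind spinning.
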NiagaraSystem /ShockwaveMace/FX/NS_ShockwaveMace_Bounce_Crackle
dependencies {
 usedDataInterfaces [&NiagaraDataInterfaceCurve, &NiagaraDataInterfaceColorCurve, &NiagaraDataInterfaceVectorCurve]
 usedDynamicInputs []
 usedModules [&SystemState]
}
emitters {
 ElectricArcs {
  emitterUpdateScript [
   &EmitterState(Inactive Response=::NewEnumerator1, Life Cycle Mode=::NewEnumerator1, Loop Behavior=::NewEnumerator1, Loop Duration=0.6, UseLoopDelay=false)
   &SpawnRate(Spawn Probability=0.7, SpawnRate=50, Use Spawn Probability=true)
  ]
  particleSpawnScript [
   /Niagara/Modules/Spawn/Initialization/InitializeParticle(
    Color=&RandomRangeLinearColor(Minimum=<rgba 1.1499 0.0729 2 1>, Maximum=<rgba 5 0.2837 4.9584 1>)
    Lifetime=&UniformRangedFloat(Minimum=0.07, Maximum=0.14)
    Module.Write Color=true
    Module.Write Lifetime=true
    Module.Write Mass=true
    Module.Write Material Random=true
    Module.Write Position=false
    Module.Write Scale=false
    Module.Write SpriteRotation=true
    Module.Write SpriteSize=true
    Position=&SimulationPosition()
    Sprite Rotation=&UniformRangedFloat(Minimum=0, Maximum=360)
    Sprite Size=&UniformRangedVector2D(Minimum=<40 40>, Maximum=<90 90>)
   )
   &DynamicMaterialParameters(
    Index 0 Param 1=$Particles.MaterialRandom
    Index 0 Param 2=&Lerp_Float(A=0.7, B=1.25, Alpha=$Particles.MaterialRandom)
    Index 0 Param 4=1
    Module.Param0WriteEnabled=true
    Module.Param1WriteEnabled=true
    Module.Param2WriteEnabled=false
    Module.Param3WriteEnabled=true
   )
   &CameraOffset(Camera Offset Amount=&RandomRangeFloat(Minimum=15, Maximum=25))
   &ShapeLocation(MaskSpawnGroup=false, Override Seed=false, Sphere Radius=10, Sphere Surface Distribution=1, UseOwnerScale=true)
  ]
  particleUpdateScript [
   &ParticleState()
   &SkeletalMeshLocation(
    Bone Sampling Mode=::NewEnumerator4
    MaskSpawnGroup=false
    Sampled Position Offset=<0 25 62>
    Skeletal Mesh=$User.NewNiagaraDataInterfaceSkeletalMesh
   )
   &SpriteRotationRate(Rotation Rate=&RandomRangeFloat(Minimum=-100, Maximum=100))
   &DynamicMaterialParameters(
    Index 0 Param 3=&FloatFromCurve(FloatCurve=DataInterface() {propertyValues {Curve {keys [<at 0.5 0 RCIM_Cubic>, {arriveTangent=3.7979, interpMode=RCIM_Cubic, leaveTangent=3.7979, tangentMode=RCTM_User, time=1, value=1}]}}})
    Module.Param0WriteEnabled=false
    Module.Param1WriteEnabled=false
    Module.Param2WriteEnabled=true
    Module.Param3WriteEnabled=false
   ) {module=DynamicMaterialParameters001}
   &ScaleSpriteSize(Scale Factor=&Vector2DFromFloat(Value=&FloatFromCurve(FloatCurve=DataInterface() {propertyValues {Curve {keys [{arriveTangent=0.3147, interpMode=RCIM_Cubic, leaveTangent=0.3147, tangentMode=RCTM_User, time=0, value=0.8}, <at 1 1>]}}})))
  ]
  rendererClasses [&NiagaraSpriteRendererProperties]
  renderers [{rendererClass=&NiagaraSpriteRendererProperties}]
 }
 Glow {
  emitterUpdateScript [
   &EmitterState(
    Inactive Response=::NewEnumerator1
    Life Cycle Mode=::NewEnumerator1
    Loop Behavior=::NewEnumerator1
    Loop Duration Mode=::NewEnumerator1
    UseLoopDelay=false
   )
   &SpawnBurst_Instantaneous(Use Spawn Probability=false, UseLoopCountLimit=false)
  ]
  particleSpawnScript [
   /Niagara/Modules/Spawn/Initialization/V2/InitializeParticle(
    Color=<rgba 5.906 0.9814 7 1>
    Lifetime=0.5
    Sprite Size Mode=::NewEnumerator3
    Uniform Sprite Size=170
    UsePositionOffset=false
    Write Material Random=true
   )
   &CameraOffset(Camera Offset Amount=20)
   &DynamicMaterialParameters(
    Index 0 Param 1=0.25
    Index 0 Param 2=2
    Index 1 Param 1=17
    Module.Param0WriteEnabled=true
    Module.Param0WriteEnabled1=true
    Module.Param1WriteEnabled=true
    Module.Param1WriteEnabled1=false
    Module.Param2WriteEnabled=true
    Module.Param2WriteEnabled1=false
    Module.Param3WriteEnabled=true
    Module.Param3WriteEnabled1=false
    Write Parameter Index 1=true
   )
  ]
  particleUpdateScript [
   &ParticleState(Kill Particles When Lifetime Has Elapsed=false)
   &ScaleSpriteSize(
    Scale Sprite Size Mode=::NewEnumerator0
    Uniform Scale Factor=&Waveform("Waveform [1]"=::NewEnumerator10, "[1] Frequency"=16, "[1] Amplitude Min/Max"=<0.6 1.1>)
   )
   &SkeletalMeshLocation(
    Bone Sampling Mode=::NewEnumerator4
    MaskSpawnGroup=false
    Sampled Position Offset=<0 25 62>
    Skeletal Mesh=$User.NewNiagaraDataInterfaceSkeletalMesh
   )
   &ScaleColor(
    Scale Alpha=&FloatFromCurve(FloatCurve=DataInterface() {propertyValues {Curve {keys [
         {arriveTangent=2.5073, interpMode=RCIM_Cubic, leaveTangent=2.5073, tangentMode=RCTM_User, time=0, value=0}
         {interpMode=RCIM_Cubic, tangentMode=RCTM_User, time=1, value=1}
        ]}}})
    ScaleA=true
    ScaleRGB=false
   )
  ]
  rendererClasses [&NiagaraSpriteRendererProperties, &NiagaraLightRendererProperties]
  renderers [
   {rendererClass=&NiagaraSpriteRendererProperties}
   {rendererClass=&NiagaraLightRendererProperties, rendererIndex=1}
  ]
 }
 OuterSphere {
  emitterUpdateScript [
   &EmitterState(Life Cycle Mode=::NewEnumerator1, Loop Behavior=::NewEnumerator1, Loop Delay=0.1, Loop Duration=0.75, UseLoopDelay=true)
   &SpawnRate(SpawnRate=75, Use Spawn Probability=false)
   &SpawnBurst_Instantaneous(Use Spawn Probability=false, UseLoopCountLimit=false) {enabled=false}
   &SpawnBurst_Instantaneous(Spawn Time=0.08, Use Spawn Probability=false, UseLoopCountLimit=false) {enabled=false, module=SpawnBurst_Instantaneous001}
   &SpawnBurst_Instantaneous(Spawn Time=0.16, Use Spawn Probability=false, UseLoopCountLimit=false) {enabled=false, module=SpawnBurst_Instantaneous002}
  ]
  particleSpawnScript [
   &InitializeParticle(
    Color=<rgba 1.5 1.5 1.5 1>
    Lifetime=&Lerp_Float(
     A=0.1
     B=0.15
     Alpha=&FloatFromCurve(
      FloatCurve=DataInterface() {propertyValues {Curve {keys [
          {interpMode=RCIM_Cubic, tangentMode=RCTM_User, time=0, value=0}
          {arriveTangent=0.891, interpMode=RCIM_Cubic, leaveTangent=0.891, tangentMode=RCTM_User, time=0.8, value=0.25}
          {interpMode=RCIM_Cubic, tangentMode=RCTM_User, time=1, value=1}
         ]}}}
      CurveIndex=&RandomRangeFloat(Minimum=0, Maximum=1)
     )
    )
    Mesh Scale=<3 3 3.2>
    Mesh Scale Mode=::NewEnumerator4
    UsePositionOffset=false
    Write Material Random=true
   )
   /SetVariables(Particles.LightPosition:NiagaraPosition=<0 0 40>) {isSetParameters=true}
   &InitialMeshOrientation(
    Mesh Axis To Orient=<0 0 1>
    Mesh Orientation Mode=::NewEnumerator2
    Rotation=&RandomRangeVector(Minimum=<0 0 0>, Maximum=<0 0 1>)
    Use Rotation Vector=true
   )
   &SkeletalMeshLocation(
    Bone Sampling Mode=::NewEnumerator4
    MaskSpawnGroup=false
    Sampled Position Offset=<0 25 62>
    Skeletal Mesh=$User.NewNiagaraDataInterfaceSkeletalMesh
   )
  ]
  particleUpdateScript [
   &ParticleState()
   /SetVariables(Particles.Position=&MakeVector(
     X=&FloatFromCurve(
      FloatCurve=DataInterface() {propertyValues {Curve {keys [
          {arriveTangent=1.4387, interpMode=RCIM_Cubic, leaveTangent=1.4387, tangentMode=RCTM_User, time=0, value=0}
          {arriveTangent=0.3784, interpMode=RCIM_Cubic, leaveTangent=0.3784, tangentMode=RCTM_User, time=1, value=1}
         ]}}}
      Scale Curve=200
     )
     Y=0
     Z=&FloatFromCurve(
      FloatCurve=DataInterface() {propertyValues {Curve {keys [
          {arriveTangent=1.4387, time=0, value=0.1}
          {arriveTangent=0.3784, interpMode=RCIM_Cubic, leaveTangent=0.3784, tangentMode=RCTM_User, time=1, value=1}
         ]}}}
      Scale Curve=0
     )
    )) {enabled=false, isSetParameters=true}
   /SetVariables(Particles.LightColor=&ColorFromCurve(ColorCurve=DataInterface() {propertyValues {
       AlphaCurve {keys [
         <at 0 1 RCIM_Cubic>
         {arriveTangent=-2.4738, interpMode=RCIM_Cubic, leaveTangent=-2.4738, tangentMode=RCTM_User, time=0.3627, value=0.7008}
         <at 1 0 RCIM_Cubic>
        ]}
       BlueCurve {keys [<at 0.1733 36.9992 RCIM_Cubic>,<at 0.647 25 RCIM_Cubic>]}
       GreenCurve {keys [<at 0.1733 1.9736 RCIM_Cubic>,<at 0.647 10.9288 RCIM_Cubic>]}
       RedCurve {keys [<at 0.1733 40 RCIM_Cubic>,<at 0.647 4.126 RCIM_Cubic>]}
      }})) {isSetParameters=true}
   &ScaleMeshSize(Scale Factor=&VectorFromFloat(Value=&FloatFromCurve(
      FloatCurve=DataInterface() {propertyValues {Curve {keys [
          {arriveTangent=0.316, interpMode=RCIM_Cubic, leaveTangent=0.316, tangentMode=RCTM_User, time=0, value=0.25}
          {arriveTangent=1.2456, interpMode=RCIM_Cubic, leaveTangent=1.2456, tangentMode=RCTM_User, time=1, value=1}
         ]}}}
      Scale Curve=0.3
     )))
   &DynamicMaterialParameters(
    Index 0 Param 1=&FloatFromCurve(
     FloatCurve=DataInterface() {propertyValues {Curve {keys [
         {arriveTangent=-2.9934, interpMode=RCIM_Cubic, leaveTangent=-2.9934, tangentMode=RCTM_User, time=0, value=1}
         {interpMode=RCIM_Cubic, leaveTangent=-1.2287, time=1, value=-0.25}
        ]}}}
     Scale Curve=2
    )
    Index 0 Param 2=1
    Index 0 Param 3=&FloatFromCurve(
     FloatCurve=DataInterface() {propertyValues {Curve {keys [
         {arriveTangent=2.2611, interpMode=RCIM_Cubic, leaveTangent=2.2611, tangentMode=RCTM_User, time=0, value=0}
         {arriveTangent=0.402, interpMode=RCIM_Cubic, leaveTangent=0.402, tangentMode=RCTM_User, time=1, value=1}
        ]}}}
     Scale Curve=5
    )
    Index 0 Param 4=&FloatFromCurve(FloatCurve=DataInterface() {propertyValues {Curve {keys [{arriveTangent=-2.6143, interpMode=RCIM_Cubic, time=0, value=1}, <at 0.7 0>]}}})
    Module.Param0WriteEnabled=true
    Module.Param1WriteEnabled=true
    Module.Param2WriteEnabled=true
    Module.Param3WriteEnabled=true
   )
   &ScaleColor(
    Scale Alpha=&FloatFromCurve(FloatCurve=DataInterface() {propertyValues {Curve {keys [
         {arriveTangent=26.1682, interpMode=RCIM_Cubic, leaveTangent=26.1682, tangentMode=RCTM_User, time=0, value=0}
         {arriveTangent=-2.2763, interpMode=RCIM_Cubic, leaveTangent=-2.2763, tangentMode=RCTM_User, time=0.07, value=1}
         {tangentMode=RCTM_User, time=1, value=0}
        ]}}})
    Scale RGB=&Lerp_Vector(
     A=<1.2 0.2 0.8>
     B=<0.2 1 1>
     Alpha=&FloatFromCurve(
      FloatCurve=DataInterface() {propertyValues {Curve {keys [{interpMode=RCIM_Cubic, tangentMode=RCTM_User, time=0.1, value=0}, {interpMode=RCIM_Cubic, tangentMode=RCTM_User, time=0.4, value=1}]}}}
      CurveIndex=&Lerp_Float(A=$Emitter.Age, B=$Particles.Age, Alpha=0.6)
     )
    )
    ScaleA=true
    ScaleRGB=true
   )
  ]
  rendererClasses [&NiagaraMeshRendererProperties]
  renderers [{rendererClass=&NiagaraMeshRendererProperties, rendererIndex=0}]
 }
 Wind {
  emitterUpdateScript [
   &EmitterState(Life Cycle Mode=::NewEnumerator1, Loop Behavior=::NewEnumerator1, Loop Delay=0.1, Loop Duration=0.75, UseLoopDelay=true)
   &SpawnRate(SpawnRate=40, Use Spawn Probability=false)
   &SpawnBurst_Instantaneous(Spawn Count=4, Use Spawn Probability=false, UseLoopCountLimit=false) {enabled=false}
  ]
  particleSpawnScript [
   &InitializeParticle(
    Color=<rgba 1 1 1 0.35>
    Lifetime Max=0.3
    Lifetime Min=0.2
    Lifetime Mode=::NewEnumerator1
    Mesh Scale=&Lerp_Vector(A=<3 3 4>, B=<6 6 6>, Alpha=&RandomRangeFloat(Minimum=0, Maximum=1))
    Mesh Scale Mode=::NewEnumerator4
    UsePositionOffset=false
    Write Material Random=true
   ) {module=InitializeParticle001}
   &InitialMeshOrientation(
    Mesh Axis To Orient=<0 0 1>
    Mesh Orientation Mode=::NewEnumerator2
    Rotation=&RandomRangeVector(Minimum=<0 0 0>, Maximum=<0 0 1>)
    Use Rotation Vector=true
   )
   &SkeletalMeshLocation(
    Bone Sampling Mode=::NewEnumerator4
    MaskSpawnGroup=false
    Sampled Position Offset=<0 25 62>
    Skeletal Mesh=$User.NewNiagaraDataInterfaceSkeletalMesh
   )
  ]
  particleUpdateScript [
   &ParticleState()
   &UpdateMeshOrientation(Rotation Rate=&RandomRangeFloat(Minimum=-1, Maximum=-2.5), Rotation Vector=<0 0 1>)
   &ScaleMeshSize(Scale Factor=&VectorFromCurve(VectorCurve=DataInterface() {propertyValues {
       XCurve {keys [
         {arriveTangent=1.9177, interpMode=RCIM_Cubic, leaveTangent=1.9177, tangentMode=RCTM_User, time=0, value=0.1}
         {arriveTangent=0.2322, interpMode=RCIM_Cubic, leaveTangent=0.2322, tangentMode=RCTM_User, time=0.8, value=1}
         {arriveTangent=-3.7145, interpMode=RCIM_Cubic, leaveTangent=-3.7145, tangentMode=RCTM_User, time=1, value=0.7}
        ]}
       YCurve {keys [
         {arriveTangent=1.9177, interpMode=RCIM_Cubic, leaveTangent=1.9177, tangentMode=RCTM_User, time=0, value=0.1}
         {arriveTangent=0.2322, interpMode=RCIM_Cubic, leaveTangent=0.2322, tangentMode=RCTM_User, time=0.8, value=1}
         {arriveTangent=-3.7145, interpMode=RCIM_Cubic, leaveTangent=-3.7145, tangentMode=RCTM_User, time=1, value=0.7}
        ]}
       ZCurve {keys [
         {arriveTangent=3.0696, interpMode=RCIM_Cubic, leaveTangent=3.0696, tangentMode=RCTM_User, time=0, value=0.5}
         {arriveTangent=-0.3, interpMode=RCIM_Cubic, leaveTangent=-0.3, time=0.35, value=0.9}
         {interpMode=RCIM_Cubic, leaveTangent=-1.6862, time=1, value=0.2}
        ]}
      }}))
   &ScaleMeshSize(Scale Factor=&VectorFromFloat(Value=0.25)) {module=ScaleMeshSize001}
   /SetVariables(Particles.Position=&AddVectorToPosition(
     Position=&Multiply_VectorByFloat(
      Vector=&ReturnMeshOrientationAxis(Mesh Orientation Axis=::ZAxis, Mesh Orientation Quat=$Particles.MeshOrientation)
      Float=&FloatFromCurve(
       FloatCurve=DataInterface() {propertyValues {Curve {keys [{arriveTangent=6.1711, interpMode=RCIM_Cubic, leaveTangent=6.1711, tangentMode=RCTM_User, time=0, value=-0.15}, <at 0.45 1 RCIM_Cubic>]}}}
       CurveIndex=$Particles.Age
       Scale Curve=&RandomRangeFloat(Minimum=500, Maximum=700)
      )
     )
     Vector=&MakeVector(
      X=0
      Y=0
      Z=&FloatFromCurve(
       FloatCurve=DataInterface() {propertyValues {Curve {keys [
           {arriveTangent=1.4387, time=0, value=0.2}
           {arriveTangent=0.3784, interpMode=RCIM_Cubic, leaveTangent=0.3784, tangentMode=RCTM_User, time=1, value=1}
          ]}}}
       Scale Curve=200
      )
     )
    )) {enabled=false, isSetParameters=true}
   &CameraOffset(Camera Offset Amount=&Lerp_Float(A=70, B=140, Alpha=$Particles.NormalizedAge))
   &ScaleColor(
    Scale Alpha=&FloatFromCurve(FloatCurve=DataInterface() {propertyValues {Curve {keys [
         {arriveTangent=8.125, interpMode=RCIM_Cubic, time=0, value=0}
         {interpMode=RCIM_Cubic, tangentMode=RCTM_User, time=0.3, value=1}
         {arriveTangent=-1.557, interpMode=RCIM_Cubic, leaveTangent=-1.557, tangentMode=RCTM_User, time=0.7791, value=0.7}
         {interpMode=RCIM_Cubic, leaveTangent=-2.3395, time=1, value=0}
        ]}}})
    Scale RGB=&MakeVectorFromLinearColorRGB(Color=&ColorFromCurve(
      ColorCurve=DataInterface() {propertyValues {
        AlphaCurve {keys []}
        BlueCurve {keys [<at 0.0485 6 RCIM_Cubic>, {arriveTangent=-19.0486, interpMode=RCIM_Cubic, leaveTangent=-19.0486, time=0.1696, value=2}, <at 0.3295 0.6458>]}
        GreenCurve {keys [
          <at 0.0485 1.018 RCIM_Cubic>
          {arriveTangent=-2.2041, interpMode=RCIM_Cubic, leaveTangent=-2.2041, time=0.1696, value=0.4856}
          <at 0.3295 0.3985>
         ]}
        RedCurve {keys [
          <at 0.0485 5.153 RCIM_Cubic>
          {arriveTangent=-16.8034, interpMode=RCIM_Cubic, leaveTangent=-16.8034, time=0.1696, value=1.4424}
          <at 0.3295 0.4299>
         ]}
       }}
      CurveIndex=$Emitter.Age
     ))
    ScaleA=true
    ScaleRGB=true
   )
   &DynamicMaterialParameters(
    Index 0 Param 1=&FloatFromCurve(FloatCurve=DataInterface() {propertyValues {Curve {keys [
         {arriveTangent=-13.4039, interpMode=RCIM_Cubic, leaveTangent=-13.4039, tangentMode=RCTM_User, time=0, value=1}
         <at 0.1 0>
         {arriveTangent=1.1111, interpMode=RCIM_Cubic, leaveTangent=1.1111, time=0.4, value=0}
         {interpMode=RCIM_Cubic, leaveTangent=16.4602, time=1, value=1}
        ]}}})
    Index 0 Param 2=&FloatFromCurve(
     FloatCurve=DataInterface() {propertyValues {Curve {keys [
         {arriveTangent=2.4843, interpMode=RCIM_Cubic, leaveTangent=2.4843, tangentMode=RCTM_User, time=0, value=0}
         {interpMode=RCIM_Cubic, leaveTangent=-0.8461, time=1, value=1}
        ]}}}
     Scale Curve=0.5
    )
    Index 0 Param 4=&FloatFromCurve(
     FloatCurve=DataInterface() {propertyValues {Curve {keys [
         {arriveTangent=0.2805, interpMode=RCIM_Cubic, time=0, value=1}
         {arriveTangent=2.4494, interpMode=RCIM_Cubic, leaveTangent=2.4494, tangentMode=RCTM_User, time=1, value=2}
        ]}}}
     Scale Curve=&RandomRangeFloat(Minimum=0.6, Maximum=0.8)
    )
    Module.Param0WriteEnabled=true
    Module.Param1WriteEnabled=true
    Module.Param2WriteEnabled=false
    Module.Param3WriteEnabled=true
   )
  ]
  rendererClasses [&NiagaraMeshRendererProperties]
  renderers [{rendererClass=&NiagaraMeshRendererProperties, rendererIndex=0}]
 }
}
summary {
 systemName=NS_ShockwaveMace_Bounce_Crackle
 emitters [
  {emitterName=ElectricArcs, rendererClasses [&NiagaraSpriteRendererProperties]}
  {emitterName=Glow, rendererClasses [&NiagaraSpriteRendererProperties, &NiagaraLightRendererProperties]}
  {emitterName=OuterSphere, rendererClasses [&NiagaraMeshRendererProperties]}
  {emitterName=Wind, rendererClasses [&NiagaraMeshRendererProperties]}
 ]
 userVariables [User.NewNiagaraDataInterfaceSkeletalMesh {
   type=NiagaraDataInterfaceSkeletalMesh
   defaultValue {
    struct=NiagaraExt_VariableValue_DataInterface
    value {dataInterface=&self:NiagaraDataInterfaceSkeletalMesh_0}
   }
  }]
}
