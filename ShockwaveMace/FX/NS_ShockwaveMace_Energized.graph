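# Niagara effect for the Shockwave Mace's energized state, combining an energy sphere, electric arcs with persistent audio, vents and smoke.
# OuterSphere pops in and persists, with color and light tinted cyan-blue to magenta by User.AmmoCount/4.
# One-shot vent mesh and sideways smoke bursts accompany startup, while User.Steam gates continuous Vents001 and Smoke001 spawning and sets the sphere's renderer visibility tag.
# User.Magnitude is declared but unreferenced.
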NiagaraSystem /ShockwaveMace/FX/NS_ShockwaveMace_Energized
dependencies {
 usedDataInterfaces [
  &NiagaraDataInterfaceCurve
  &NiagaraDataInterfaceColorCurve
  &NiagaraDataInterfaceAudioPlayer
  &NiagaraDataInterfaceVectorCurve
 ]
 usedDynamicInputs []
 usedModules [&SystemState]
}
emitters {
 ElectricArcs {
  emitterUpdateScript [
   &EmitterState(Life Cycle Mode=::NewEnumerator1, Loop Behavior=::NewEnumerator1, Loop Duration Mode=::NewEnumerator1, UseLoopDelay=false)
   &SpawnBurst_Instantaneous(Spawn Count=5, Spawn Time=0.05, Use Spawn Probability=false, UseLoopCountLimit=false)
   &SpawnBurst_Instantaneous(Spawn Count=2, Spawn Time=0.1, Use Spawn Probability=false, UseLoopCountLimit=false) {module=SpawnBurst_Instantaneous001}
   &SpawnRate(Spawn Probability=0.25, SpawnRate=4, Use Spawn Probability=true)
  ]
  particleSpawnScript [
   /Niagara/Modules/Spawn/Initialization/InitializeParticle(
    Color=&RandomRangeLinearColor(Minimum=<rgba 4 0.7052 3.8857 1>, Maximum=<rgba 6 1.4004 4.6812 1>)
    Lifetime=&UniformRangedFloat(Minimum=0.1, Maximum=0.15)
    Module.Write Color=true
    Module.Write Lifetime=true
    Module.Write Mass=true
    Module.Write Material Random=true
    Module.Write Position=false
    Module.Write Scale=false
    Module.Write SpriteRotation=true
    Module.Write SpriteSize=true
    Position=&SimulationPosition()
    Sprite Rotation=&UniformRangedFloat(Minimum=0, Maximum=360)
    Sprite Size=&UniformRangedVector2D(Minimum=<40 40>, Maximum=<80 80>)
   )
   &DynamicMaterialParameters(
    Index 0 Param 1=$Particles.MaterialRandom
    Index 0 Param 2=&Lerp_Float(A=0.7, B=1.25, Alpha=$Particles.MaterialRandom)
    Index 0 Param 4=1
    Module.Param0WriteEnabled=true
    Module.Param1WriteEnabled=true
    Module.Param2WriteEnabled=false
    Module.Param3WriteEnabled=true
   )
   &CameraOffset(Camera Offset Amount=12)
   &ShapeLocation(
    MaskSpawnGroup=false
    Offset=<5 0 0>
    Offset Mode=::NewEnumerator0
    Override Seed=false
    Sphere Radius=10
    Sphere Surface Distribution=1
    UseOwnerScale=true
   )
  ]
  particleUpdateScript [
   &ParticleState()
   &SpriteRotationRate(Rotation Rate=&RandomRangeFloat(Minimum=-100, Maximum=100))
   &DynamicMaterialParameters(
    Index 0 Param 3=&FloatFromCurve(FloatCurve=DataInterface() {propertyValues {Curve {keys [<at 0.5 0 RCIM_Cubic>, {arriveTangent=3.7979, interpMode=RCIM_Cubic, leaveTangent=3.7979, tangentMode=RCTM_User, time=1, value=1}]}}})
    Index 1 Param 2=0.7
    Index 1 Param 4=500
    Module.Param0WriteEnabled=false
    Module.Param0WriteEnabled1=true
    Module.Param1WriteEnabled=false
    Module.Param1WriteEnabled1=true
    Module.Param2WriteEnabled=true
    Module.Param2WriteEnabled1=true
    Module.Param3WriteEnabled=false
    Module.Param3WriteEnabled1=true
    Write Parameter Index 1=true
   ) {module=DynamicMaterialParameters001}
   &ScaleSpriteSize(Scale Factor=&Vector2DFromFloat(Value=&FloatFromCurve(FloatCurve=DataInterface() {propertyValues {Curve {keys [{arriveTangent=0.3147, interpMode=RCIM_Cubic, leaveTangent=0.3147, tangentMode=RCTM_User, time=0, value=0.8}, <at 1 1>]}}})))
   /Niagara/Modules/Audio/PlayPersistentAudio(Enable Fade In / Out=true, Fade In Time=0, Fade Out Time=2)
  ]
  rendererClasses [&NiagaraSpriteRendererProperties]
  renderers [{rendererClass=&NiagaraSpriteRendererProperties}]
 }
 OuterSphere {
  emitterUpdateScript [
   &EmitterState(
    Inactive Response=::NewEnumerator1
    Life Cycle Mode=::NewEnumerator1
    Loop Behavior=::NewEnumerator1
    Loop Duration Mode=::NewEnumerator1
    UseLoopDelay=false
   )
   &SpawnBurst_Instantaneous(Use Spawn Probability=false, UseLoopCountLimit=false)
  ]
  particleSpawnScript [
   /Niagara/Modules/Spawn/Initialization/V2/InitializeParticle(
    Color=<rgba 2 2 2 0.1>
    Lifetime=0.6
    Mass=35
    Mass Mode=::NewEnumerator1
    Mesh Scale=<1 1.04 1>
    Mesh Scale Mode=::NewEnumerator4
    Position Offset=<4 0 0>
    Sprite Size Mode=::NewEnumerator3
    Uniform Sprite Size=70
    UsePositionOffset=true
    Write Material Random=true
   )
   /SetVariables(Particles.LightPosition:NiagaraPosition=<15 0 0>) {isSetParameters=true}
   &CameraOffset(Camera Offset Amount=1)
   &InitialMeshOrientation(Mesh Orientation Mode=::NewEnumerator2, Use Rotation Vector=false)
   &DynamicMaterialParameters(
    Index 0 Param 2=1
    Index 0 Param 4=1
    Index 2 Param 1=1.5
    Index 2 Param 2=0.5
    Index 3 Param 1=2
    Index 3 Param 2=0.5
    Module.Param0WriteEnabled=true
    Module.Param0WriteEnabled2=true
    Module.Param0WriteEnabled3=true
    Module.Param1WriteEnabled=true
    Module.Param1WriteEnabled2=true
    Module.Param1WriteEnabled3=true
    Module.Param2WriteEnabled=false
    Module.Param2WriteEnabled2=true
    Module.Param2WriteEnabled3=false
    Module.Param3WriteEnabled=true
    Module.Param3WriteEnabled2=true
    Module.Param3WriteEnabled3=false
    Write Parameter Index 2=true
    Write Parameter Index 3=true
   )
   /SetVariables(Particles.DecalPosition:NiagaraPosition=<25 0 0>) {isSetParameters=true}
  ]
  particleUpdateScript [
   &ParticleState(Kill Particles When Lifetime Has Elapsed=false)
   &ScaleSpriteSize(
    Scale Sprite Size Mode=::NewEnumerator0
    Uniform Scale Factor=&FloatFromCurve(FloatCurve=DataInterface() {propertyValues {Curve {keys [
         {arriveTangent=9.4929, interpMode=RCIM_Cubic, leaveTangent=9.4929, tangentMode=RCTM_User, time=0, value=0.5}
         {arriveTangent=0.1291, interpMode=RCIM_Cubic, leaveTangent=0.1291, tangentMode=RCTM_User, time=0.1444, value=1.1}
         {arriveTangent=0.2126, interpMode=RCIM_Cubic, leaveTangent=0.2126, tangentMode=RCTM_User, time=0.4981, value=0.9647}
         {tangentMode=RCTM_User, time=0.8111, value=1}
        ]}}})
   )
   &ScaleMeshSize(Scale Factor=&VectorFromFloat(Value=&FloatFromCurve(FloatCurve=DataInterface() {propertyValues {Curve {keys [
          {arriveTangent=9.4929, interpMode=RCIM_Cubic, leaveTangent=9.4929, tangentMode=RCTM_User, time=0, value=0.5}
          {arriveTangent=0.1291, interpMode=RCIM_Cubic, leaveTangent=0.1291, tangentMode=RCTM_User, time=0.1444, value=1.05}
          {arriveTangent=0.2126, interpMode=RCIM_Cubic, leaveTangent=0.2126, tangentMode=RCTM_User, time=0.4981, value=0.9647}
          {tangentMode=RCTM_User, time=0.8111, value=1}
         ]}}})))
   /SetVariables(Particles.LightColor=&MultiplyLinearColor(
     Linear Color=&ColorFromCurve(
      ColorCurve=DataInterface() {propertyValues {
        AlphaCurve {keys [<at 0 1 RCIM_Cubic>]}
        BlueCurve {keys [<at 0 1 RCIM_Cubic>,<at 1 1 RCIM_Cubic>]}
        GreenCurve {keys [<at 0 0.3095 RCIM_Cubic>,<at 1 0.0273 RCIM_Cubic>]}
        RedCurve {keys [<at 0 0 RCIM_Cubic>,<at 1 0.6867 RCIM_Cubic>]}
       }}
      CurveIndex=&Divide_Float(A=$User.AmmoCount, B=4)
     )
     Scale Factor=&MakeVector4_FromFloat(Float=&FloatFromCurve(
       FloatCurve=DataInterface() {propertyValues {Curve {keys [
           {arriveTangent=20.4225, interpMode=RCIM_Cubic, time=0, value=1}
           {arriveTangent=5, interpMode=RCIM_Cubic, leaveTangent=5, time=0.07, value=3.5}
           {arriveTangent=-5.2324, interpMode=RCIM_Cubic, leaveTangent=-5.2324, tangentMode=RCTM_User, time=0.3, value=2.5}
           {tangentMode=RCTM_User, time=1, value=1}
          ]}}}
       Scale Curve=7
      ))
    )) {isSetParameters=true}
   &ScaleColor(
    Scale Alpha=&FloatFromCurve(FloatCurve=DataInterface() {propertyValues {Curve {keys [
         {arriveTangent=20.9313, interpMode=RCIM_Cubic, time=0, value=0}
         {arriveTangent=-0.8703, interpMode=RCIM_Cubic, leaveTangent=-0.8703, tangentMode=RCTM_User, time=0.07, value=1.25}
         {tangentMode=RCTM_User, time=1, value=1}
        ]}}})
    Scale RGB=&MakeVectorFromLinearColorRGB(Color=&MultiplyLinearColor(
      Linear Color=&ColorFromCurve(
       ColorCurve=DataInterface() {propertyValues {
         AlphaCurve {keys [<at 0 1 RCIM_Cubic>]}
         BlueCurve {keys [<at 0 1 RCIM_Cubic>,<at 1 1 RCIM_Cubic>]}
         GreenCurve {keys [<at 0 0.3095 RCIM_Cubic>,<at 1 0.0273 RCIM_Cubic>]}
         RedCurve {keys [<at 0 0 RCIM_Cubic>,<at 1 0.6867 RCIM_Cubic>]}
        }}
       CurveIndex=&Divide_Float(A=$User.AmmoCount, B=4)
      )
      Scale Factor=&MakeVector4_FromFloat(Float=&FloatFromCurve(FloatCurve=DataInterface() {propertyValues {Curve {keys [
            {arriveTangent=20.4225, interpMode=RCIM_Cubic, time=0, value=1}
            {arriveTangent=5, interpMode=RCIM_Cubic, leaveTangent=5, time=0.07, value=3.5}
            {arriveTangent=-5.2324, interpMode=RCIM_Cubic, leaveTangent=-5.2324, tangentMode=RCTM_User, time=0.3, value=2.5}
            {tangentMode=RCTM_User, time=1, value=1}
           ]}}}))
     ))
    ScaleA=true
    ScaleRGB=true
   )
   /SetVariables(Particles.DecalColor:LinearColor=&MultiplyLinearColor(
     Linear Color=$Particles.Color
     Scale Factor=&MakeFloat4FromVectorAndFloat("Vector (RGB)"=&VectorFromFloat(Value=5), "Float (Alpha)"=1)
    )) {isSetParameters=true}
   /SetVariables(Particles.VisibilityTag=&MakeCustomIntFromBool(Bool=$User.Steam, True INT=1, False INT=0)) {isSetParameters=true}
  ]
  rendererClasses [
   &NiagaraMeshRendererProperties
   &NiagaraLightRendererProperties
   &NiagaraSpriteRendererProperties
  ]
  renderers [
   {rendererClass=&NiagaraMeshRendererProperties, rendererIndex=0}
   {rendererClass=&NiagaraLightRendererProperties, rendererIndex=1}
   {rendererClass=&NiagaraSpriteRendererProperties, rendererIndex=2}
   {rendererClass=&NiagaraSpriteRendererProperties, rendererIndex=3}
  ]
 }
 Smoke {
  emitterUpdateScript [
   &EmitterState(
    Enable Distance Culling=true
    Life Cycle Mode=::NewEnumerator1
    Loop Behavior=::NewEnumerator1
    Loop Duration=0.4
    MaxDistance=2500
    Module.bUseMaxDistance=true
    Module.bUseMinDistance=false
    Scalability Mode=::NewEnumerator1
    UseLoopDelay=false
   )
   &SpawnRate(
    SpawnRate=&FloatFromCurve(
     FloatCurve=DataInterface() {propertyValues {Curve {keys [<at 0.1 1 RCIM_Cubic>, {arriveTangent=-2.7102, interpMode=RCIM_Cubic, leaveTangent=-2.7102, tangentMode=RCTM_User, time=0.75, value=0.25}]}}}
     CurveIndex=$Emitter.NormalizedLoopAge
     Scale Curve=75
    )
    Use Spawn Probability=false
   )
   &SpawnBurst_Instantaneous(Spawn Count=15, Use Spawn Probability=false)
  ]
  particleSpawnScript [
   &InitializeParticle(
    Color Maximum=<rgba 0.6981 0.4958 0.7 0.4>
    Color Minimum=<rgba 0.5355 0.45 0.6 0.25>
    Color Mode=::NewEnumerator2
    Lifetime=&RandomRangeFloat(Minimum=0.25, Maximum=0.35)
    Sprite Rotation Angle=&RandomRangeFloat(Minimum=-0.15, Maximum=0.15)
    Sprite Rotation Mode=::NewEnumerator1
    Sprite Size Mode=::NewEnumerator1
    Uniform Sprite Size Max=60
    Uniform Sprite Size Min=40
    UsePositionOffset=false
    Write Material Random=true
   ) {module=InitializeParticle001}
   &SphereLocation(
    MaskSpawnGroup=false
    Override Seed=false
    Sphere Radius=3
    Surface Only=false
    UseNonUniformScale=false
    UseOffset=false
    UseSphereOrientation=false
   ) {module=SphereLocation001}
   &AddVelocityInCone(
    Cone Angle=20
    Cone Axis=&MakeCustomVectorFromBool(Bool=&RandomBool(), True Vector=<0 1 0>, False Vector=<0 -1 0>)
    Module.Use Velocity Falloff On Cone Axis=false
    Velocity Strength=&RandomRangeFloat(Minimum=200, Maximum=300)
   )
  ]
  particleUpdateScript [
   &ParticleState()
   &Drag(Drag=&RandomRangeFloat(Minimum=2, Maximum=3), Use Linear Drag=true, Use Rotational Drag=false)
   &SpriteRotationRate(Rotation Rate=&Multiply_Float(
     A=&Lerp_Float(A=-300, B=300, Alpha=$Particles.MaterialRandom)
     B=&FloatFromCurve(FloatCurve=DataInterface() {propertyValues {Curve {keys [
          {arriveTangent=-4.4572, interpMode=RCIM_Cubic, leaveTangent=-4.4572, tangentMode=RCTM_User, time=0, value=1}
          {arriveTangent=-1.0628, interpMode=RCIM_Cubic, leaveTangent=-1.0628, tangentMode=RCTM_User, time=0.2, value=0.4}
          <at 1.0011 0.01>
         ]}}})
    ))
   &SolveForcesAndVelocity(Clamp Velocity=false, Limit Acceleration=false)
   &CameraOffset(Camera Offset Amount=&Multiply_Float(
     A=0.6
     B=/Niagara/DynamicInputs/Math/CalculateParticleRadius(
      Radius Calculation Type:ENiagaraCollisionRadiusOptions=::NewEnumerator0
      Particle Radius:NiagaraFloat=1
      Mesh Dimensions:Vector3f=<1 1 1>
      Method for Calculating Particle Radius:ENiagaraCalculateRadiusOptions=::NewEnumerator2
     )
    ))
   &ScaleSpriteSize(Scale Factor=&Vector2DFromFloat(Value=&FloatFromCurve(FloatCurve=DataInterface() {propertyValues {Curve {keys [
          {arriveTangent=2.7007, interpMode=RCIM_Cubic, leaveTangent=2.7007, tangentMode=RCTM_User, time=0, value=0.15}
          {arriveTangent=0.8163, interpMode=RCIM_Cubic, leaveTangent=0.8163, tangentMode=RCTM_User, time=0.3314, value=0.65}
          {arriveTangent=0.424, interpMode=RCIM_Cubic, leaveTangent=0.424, tangentMode=RCTM_User, time=1, value=1}
         ]}}})))
   &ScaleColor(
    Scale Alpha=&FloatFromCurve(FloatCurve=DataInterface() {propertyValues {Curve {keys [
         {arriveTangent=12.69, interpMode=RCIM_Cubic, leaveTangent=12.69, tangentMode=RCTM_User, time=0, value=0.05}
         <at 0.2 1 RCIM_Cubic>
         {interpMode=RCIM_Cubic, leaveTangent=-3.4295, time=1, value=0}
        ]}}})
    Scale RGB=&VectorFromCurve(
     VectorCurve=DataInterface() {propertyValues {
       XCurve {keys [{arriveTangent=-17.4689, interpMode=RCIM_Cubic, leaveTangent=-17.4689, tangentMode=RCTM_User, time=0, value=5}, <at 0.5 0.7>]}
       YCurve {keys [{arriveTangent=-3.401, interpMode=RCIM_Cubic, leaveTangent=-3.401, tangentMode=RCTM_User, time=0, value=1.5}, <at 0.5 0.7>]}
       ZCurve {keys [{arriveTangent=-6.0501, interpMode=RCIM_Cubic, time=0, value=1}, <at 0.5 0.7>]}
      }}
     CurveIndex=$Emitter.Age
    )
    ScaleA=true
    ScaleRGB=true
   )
   /SetVariables(Particles.SubImageIndex=&Lerp_Float(A=20, B=63, Alpha=$Particles.NormalizedAge)) {isSetParameters=true, module=SetVariables_#1}
  ]
  rendererClasses [&NiagaraSpriteRendererProperties]
  renderers [{rendererClass=&NiagaraSpriteRendererProperties}]
 }
 Smoke001 {
  emitterUpdateScript [
   &EmitterState(Enable Distance Culling=true, MaxDistance=2500, Module.bUseMaxDistance=true, Module.bUseMinDistance=false, Scalability Mode=::NewEnumerator1)
   &SpawnRate(
    Spawn Probability=&MakeCustomFloatFromBool(Bool=$User.Steam, True Float=1, False Float=0)
    SpawnRate=40
    Use Spawn Probability=true
   )
  ]
  particleSpawnScript [
   &InitializeParticle(
    Color Maximum=<rgba 0.6981 0.4958 0.7 0.4>
    Color Minimum=<rgba 0.5355 0.45 0.6 0.25>
    Color Mode=::NewEnumerator2
    Lifetime=&RandomRangeFloat(Minimum=0.25, Maximum=0.35)
    Sprite Rotation Angle=&RandomRangeFloat(Minimum=-0.15, Maximum=0.15)
    Sprite Rotation Mode=::NewEnumerator1
    Sprite Size Mode=::NewEnumerator1
    Uniform Sprite Size Max=50
    Uniform Sprite Size Min=30
    UsePositionOffset=false
    Write Material Random=true
   ) {module=InitializeParticle001}
   &SphereLocation(
    MaskSpawnGroup=false
    Override Seed=false
    Sphere Radius=2
    Surface Only=false
    UseNonUniformScale=false
    UseOffset=false
    UseSphereOrientation=false
   ) {module=SphereLocation001}
   &AddVelocityInCone(
    Cone Angle=20
    Cone Axis=&MakeCustomVectorFromBool(Bool=&RandomBool(), True Vector=<0 1 0>, False Vector=<0 -1 0>)
    Module.Use Velocity Falloff On Cone Axis=false
    Velocity Strength=&RandomRangeFloat(Minimum=140, Maximum=220)
   )
  ]
  particleUpdateScript [
   &ParticleState()
   &Drag(Drag=&RandomRangeFloat(Minimum=2, Maximum=3), Use Linear Drag=true, Use Rotational Drag=false)
   &SpriteRotationRate(Rotation Rate=&Multiply_Float(
     A=&Lerp_Float(A=-300, B=300, Alpha=$Particles.MaterialRandom)
     B=&FloatFromCurve(FloatCurve=DataInterface() {propertyValues {Curve {keys [
          {arriveTangent=-4.4572, interpMode=RCIM_Cubic, leaveTangent=-4.4572, tangentMode=RCTM_User, time=0, value=1}
          {arriveTangent=-1.0628, interpMode=RCIM_Cubic, leaveTangent=-1.0628, tangentMode=RCTM_User, time=0.2, value=0.4}
          <at 1.0011 0.01>
         ]}}})
    ))
   &SolveForcesAndVelocity(Clamp Velocity=false, Limit Acceleration=false)
   &CameraOffset(Camera Offset Amount=&Multiply_Float(
     A=0.6
     B=&CalculateParticleRadius(
      Radius Calculation Type:ENiagaraCollisionRadiusOptions=::NewEnumerator0
      Particle Radius:NiagaraFloat=1
      Mesh Dimensions:Vector3f=<1 1 1>
      Method for Calculating Particle Radius:ENiagaraCalculateRadiusOptions=::NewEnumerator2
     )
    ))
   &ScaleSpriteSize(Scale Factor=&Vector2DFromFloat(Value=&FloatFromCurve(FloatCurve=DataInterface() {propertyValues {Curve {keys [
          {arriveTangent=2.7007, interpMode=RCIM_Cubic, leaveTangent=2.7007, tangentMode=RCTM_User, time=0, value=0.15}
          {arriveTangent=0.8163, interpMode=RCIM_Cubic, leaveTangent=0.8163, tangentMode=RCTM_User, time=0.3314, value=0.65}
          {arriveTangent=0.424, interpMode=RCIM_Cubic, leaveTangent=0.424, tangentMode=RCTM_User, time=1, value=1}
         ]}}})))
   &ScaleColor(
    Scale Alpha=&FloatFromCurve(FloatCurve=DataInterface() {propertyValues {Curve {keys [
         {arriveTangent=12.69, interpMode=RCIM_Cubic, leaveTangent=12.69, tangentMode=RCTM_User, time=0, value=0.05}
         <at 0.2 1 RCIM_Cubic>
         {interpMode=RCIM_Cubic, leaveTangent=-3.4295, time=1, value=0}
        ]}}})
    Scale RGB=&VectorFromCurve(
     VectorCurve=DataInterface() {propertyValues {
       XCurve {keys [{arriveTangent=-17.4689, interpMode=RCIM_Cubic, leaveTangent=-17.4689, tangentMode=RCTM_User, time=0, value=5}, <at 0.5 0.7>]}
       YCurve {keys [{arriveTangent=-3.401, interpMode=RCIM_Cubic, leaveTangent=-3.401, tangentMode=RCTM_User, time=0, value=1.5}, <at 0.5 0.7>]}
       ZCurve {keys [{arriveTangent=-6.0501, interpMode=RCIM_Cubic, time=0, value=1}, <at 0.5 0.7>]}
      }}
     CurveIndex=$Emitter.Age
    )
    ScaleA=true
    ScaleRGB=true
   )
   /SetVariables(Particles.SubImageIndex=&Lerp_Float(A=20, B=63, Alpha=$Particles.NormalizedAge)) {isSetParameters=true, module=SetVariables_#1}
  ]
  rendererClasses [&NiagaraSpriteRendererProperties]
  renderers [{rendererClass=&NiagaraSpriteRendererProperties}]
 }
 Vents {
  emitterUpdateScript [
   &EmitterState(
    Inactive Response=::NewEnumerator1
    Life Cycle Mode=::NewEnumerator1
    Loop Behavior=::NewEnumerator1
    Loop Duration Mode=::NewEnumerator1
    UseLoopDelay=false
   )
   &SpawnBurst_Instantaneous(Spawn Count=1, Use Spawn Probability=false)
  ]
  particleSpawnScript [
   &InitializeParticle(
    Color=<rgba 2 0.4959 0.4363 1>
    Lifetime=0.7
    Mass=0
    Mass Mode=::NewEnumerator1
    Mesh Scale Mode=::NewEnumerator4
    Sprite Size Mode=::NewEnumerator3
    Uniform Sprite Size=17
    UsePositionOffset=true
    Write Material Random=true
   )
   &InitialMeshOrientation(Mesh Axis To Orient=<0 1 0>, Mesh Orientation Mode=::NewEnumerator2, Use Rotation Vector=false)
  ]
  particleUpdateScript [
   &ParticleState()
   &ScaleColor(
    Scale Alpha=&FloatFromCurve(FloatCurve=DataInterface() {propertyValues {Curve {keys [{arriveTangent=40.829, interpMode=RCIM_Cubic, leaveTangent=40.829, tangentMode=RCTM_User, time=0, value=0}, <at 0.05 1 RCIM_Cubic>, <at 1 0>]}}})
    ScaleA=true
    ScaleRGB=false
   )
   &DynamicMaterialParameters(
    Index 0 Param 1=&FloatFromCurve(FloatCurve=DataInterface() {propertyValues {Curve {keys [
         {arriveTangent=1.9352, interpMode=RCIM_Cubic, leaveTangent=1.9352, tangentMode=RCTM_User, time=0, value=0.1}
         {arriveTangent=0.3565, interpMode=RCIM_Cubic, leaveTangent=0.3565, tangentMode=RCTM_User, time=1, value=1}
        ]}}})
    Index 0 Param 2=3
    Index 1 Param 2=&FloatFromCurve(
     FloatCurve=DataInterface() {propertyValues {Curve {keys [{interpMode=RCIM_Cubic, tangentMode=RCTM_User, time=0, value=1}, {interpMode=RCIM_Cubic, tangentMode=RCTM_User, time=1, value=0}]}}}
     Scale Curve=3
    )
    Module.Param0WriteEnabled=true
    Module.Param0WriteEnabled1=true
    Module.Param1WriteEnabled=true
    Module.Param1WriteEnabled1=true
    Module.Param2WriteEnabled=false
    Module.Param2WriteEnabled1=false
    Module.Param3WriteEnabled=false
    Module.Param3WriteEnabled1=false
    Write Parameter Index 1=true
   ) {module=DynamicMaterialParameters001}
  ]
  rendererClasses [&NiagaraMeshRendererProperties]
  renderers [{rendererClass=&NiagaraMeshRendererProperties, rendererIndex=0}]
 }
 Vents001 {
  emitterUpdateScript [
   &EmitterState()
   &SpawnRate(
    Spawn Probability=&MakeCustomFloatFromBool(Bool=$User.Steam, True Float=1, False Float=0)
    SpawnRate=2
    Use Spawn Probability=true
   )
  ]
  particleSpawnScript [
   &InitializeParticle(
    Color=<rgba 2 0.4959 0.4363 1>
    Mass=0
    Mass Mode=::NewEnumerator1
    Mesh Scale Mode=::NewEnumerator4
    Sprite Size Mode=::NewEnumerator3
    Uniform Sprite Size=17
    UsePositionOffset=true
    Write Material Random=true
   )
   &InitialMeshOrientation(Mesh Axis To Orient=<0 1 0>, Mesh Orientation Mode=::NewEnumerator2, Use Rotation Vector=false)
  ]
  particleUpdateScript [
   &ParticleState()
   &ScaleColor(Scale Alpha=0.2, ScaleA=true, ScaleRGB=false)
   &DynamicMaterialParameters(
    Index 0 Param 1=&FloatFromCurve(
     FloatCurve=DataInterface() {propertyValues {Curve {keys [
         {arriveTangent=1.9352, interpMode=RCIM_Cubic, leaveTangent=1.9352, tangentMode=RCTM_User, time=0, value=0.1}
         {arriveTangent=0.3565, interpMode=RCIM_Cubic, leaveTangent=0.3565, tangentMode=RCTM_User, time=1, value=1}
        ]}}}
     CurveIndex=$System.CurrentLoopDuration
    )
    Index 0 Param 2=2
    Index 1 Param 2=1
    Module.Param0WriteEnabled=true
    Module.Param0WriteEnabled1=true
    Module.Param1WriteEnabled=true
    Module.Param1WriteEnabled1=true
    Module.Param2WriteEnabled=false
    Module.Param2WriteEnabled1=false
    Module.Param3WriteEnabled=false
    Module.Param3WriteEnabled1=false
    Write Parameter Index 1=true
   ) {module=DynamicMaterialParameters001}
  ]
  rendererClasses [&NiagaraMeshRendererProperties]
  renderers [{rendererClass=&NiagaraMeshRendererProperties, rendererIndex=0}]
 }
}
summary {
 systemName=NS_ShockwaveMace_Energized
 emitters [
  {
   emitterName=OuterSphere
   rendererClasses [
    &NiagaraMeshRendererProperties
    &NiagaraLightRendererProperties
    &NiagaraSpriteRendererProperties
   ]
  }
  {emitterName=ElectricArcs, rendererClasses [&NiagaraSpriteRendererProperties]}
  {emitterName=Vents, rendererClasses [&NiagaraMeshRendererProperties]}
  {emitterName=Smoke, rendererClasses [&NiagaraSpriteRendererProperties]}
  {emitterName=Vents001, rendererClasses [&NiagaraMeshRendererProperties]}
  {emitterName=Smoke001, rendererClasses [&NiagaraSpriteRendererProperties]}
 ]
 userVariables [
  User.AmmoCount {type=NiagaraFloat, defaultValue=0, description=""}
  User.Magnitude {type=NiagaraFloat, defaultValue=0, description=""}
  User.Steam {type=NiagaraBool, defaultValue=true}
 ]
}
