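# Niagara effect for the Shockwave Mace slam swing: a magenta glow on the weapon plus swirling wind meshes.
# Glow is a flickering sprite and light kept at an offset on the user-supplied skeletal mesh.
# Wind bursts three staggered, randomly flipped spinning meshes along a descending path driven by system loop age, with growing scale, color ramps and animated material parameters.
# A disabled Empty emitter holds an unused sprite/ribbon trail sampled from the same mesh.
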NiagaraSystem /ShockwaveMace/FX/NS_ShockwaveMace_Slam_Swing
dependencies {
 usedDataInterfaces [&NiagaraDataInterfaceCurve, &NiagaraDataInterfaceVectorCurve, &NiagaraDataInterfaceColorCurve]
 usedDynamicInputs []
 usedModules [&SystemState]
}
emitters {
 Empty {
  bEnabled=false
  emitterUpdateScript [
   &EmitterState()
   &SpawnRate(SpawnRate=50, Use Spawn Probability=false)
  ]
  particleSpawnScript [
   &InitializeParticle(
    Color Mode=::NewEnumerator0
    Lifetime=0.5
    Ribbon Width=25
    Ribbon Width Mode=::NewEnumerator1
    Sprite Size Mode=::NewEnumerator3
    UsePositionOffset=false
    Write Material Random=true
   )
   &SkeletalMeshLocation(
    Bone Sampling Mode=::NewEnumerator4
    MaskSpawnGroup=false
    Sampled Position Offset=<0 25 62>
    Skeletal Mesh=$User.NewNiagaraDataInterfaceSkeletalMesh
   )
  ]
  particleUpdateScript [&ParticleState()]
  rendererClasses [&NiagaraSpriteRendererProperties, &NiagaraRibbonRendererProperties]
  renderers [
   {rendererClass=&NiagaraSpriteRendererProperties}
   {rendererClass=&NiagaraRibbonRendererProperties, rendererIndex=1}
  ]
 }
 Glow {
  emitterUpdateScript [
   &EmitterState()
   &SpawnBurst_Instantaneous(Use Spawn Probability=false, UseLoopCountLimit=false)
  ]
  particleSpawnScript [
   &InitializeParticle(
    Color=<rgba 5.906 0.9814 7 1>
    Lifetime=0.25
    Sprite Size Mode=::NewEnumerator3
    Uniform Sprite Size=150
    UsePositionOffset=false
    Write Material Random=true
   )
   &CameraOffset(Camera Offset Amount=20)
   &DynamicMaterialParameters(
    Index 0 Param 1=0.25
    Index 0 Param 2=2
    Index 1 Param 1=17
    Module.Param0WriteEnabled=true
    Module.Param0WriteEnabled1=true
    Module.Param1WriteEnabled=true
    Module.Param1WriteEnabled1=false
    Module.Param2WriteEnabled=true
    Module.Param2WriteEnabled1=false
    Module.Param3WriteEnabled=true
    Module.Param3WriteEnabled1=false
    Write Parameter Index 1=true
   )
  ]
  particleUpdateScript [
   &ParticleState()
   &ScaleSpriteSize(
    Scale Sprite Size Mode=::NewEnumerator0
    Uniform Scale Factor=&Waveform("Waveform [1]"=::NewEnumerator10, "[1] Frequency"=16, "[1] Amplitude Min/Max"=<0.6 1.1>)
   )
   &SkeletalMeshLocation(
    Bone Sampling Mode=::NewEnumerator4
    MaskSpawnGroup=false
    Sampled Position Offset=<0 25 62>
    Skeletal Mesh=$User.NewNiagaraDataInterfaceSkeletalMesh
   )
   &ScaleColor(
    Scale Alpha=&FloatFromCurve(FloatCurve=DataInterface() {propertyValues {Curve {keys [
         {arriveTangent=2.5073, interpMode=RCIM_Cubic, leaveTangent=2.5073, tangentMode=RCTM_User, time=0, value=0}
         {interpMode=RCIM_Cubic, tangentMode=RCTM_User, time=0.8, value=1}
         <at 1 0>
        ]}}})
    ScaleA=true
    ScaleRGB=false
   )
  ]
  rendererClasses [&NiagaraSpriteRendererProperties, &NiagaraLightRendererProperties]
  renderers [
   {rendererClass=&NiagaraSpriteRendererProperties}
   {rendererClass=&NiagaraLightRendererProperties, rendererIndex=1}
  ]
 }
 Wind {
  emitterUpdateScript [
   &EmitterState(Enable Distance Culling=true, MaxDistance=6000, Module.bUseMaxDistance=true, Module.bUseMinDistance=false, Scalability Mode=::NewEnumerator1)
   &SpawnBurst_Instantaneous(Use Spawn Probability=false, UseLoopCountLimit=false)
   &SpawnBurst_Instantaneous(Spawn Time=0.04, Use Spawn Probability=false, UseLoopCountLimit=false) {module=SpawnBurst_Instantaneous002}
   &SpawnBurst_Instantaneous(Spawn Time=0.075, Use Spawn Probability=false, UseLoopCountLimit=false) {module=SpawnBurst_Instantaneous001}
  ]
  particleSpawnScript [
   /SetVariables(Particles.FlipSide:NiagaraBool=&RandomBool()) {isSetParameters=true}
   &InitializeParticle(
    Color=<rgba 2 2 2 0.5>
    Lifetime=&UniformRangedFloat(Minimum=0.14, Maximum=0.16)
    Mesh Scale=&Lerp_Vector(A=<2.7 2.7 1.5>, B=<3 3 3>, Alpha=&RandomRangeFloat(Minimum=0, Maximum=1))
    Mesh Scale Mode=::NewEnumerator4
    UsePositionOffset=false
    Write Material Random=true
   ) {module=InitializeParticle001}
   &InitialMeshOrientation(
    Mesh Axis To Orient=&MakeCustomVectorFromBool(Bool=$Particles.FlipSide, True Vector=<0 0 1>, False Vector=<0 0 -1>)
    Mesh Orientation Mode=::NewEnumerator2
    Rotation=&RandomRangeVector(Minimum=<0 0 -0.3>, Maximum=<0 0 -0.27>)
    Use Rotation Vector=true
   )
  ]
  particleUpdateScript [
   &ParticleState()
   &UpdateMeshOrientation(
    Rotation Rate=&Multiply_Float(
     A=&MakeCustomFloatFromBool(Bool=$Particles.FlipSide, True Float=-1, False Float=1)
     B=&RandomRangeFloat(Minimum=3.5, Maximum=3.8)
    )
    Rotation Vector=<0 0 1>
   )
   /SetVariables(Particles.Position=&MakeVector(
     X=-15
     Y=&FloatFromCurve(
      FloatCurve=DataInterface() {propertyValues {Curve {keys [
          {arriveTangent=3.2328, interpMode=RCIM_Cubic, leaveTangent=3.2328, tangentMode=RCTM_User, time=0, value=-1}
          {interpMode=RCIM_Cubic, leaveTangent=1.2424, time=1, value=0.4}
         ]}}}
      CurveIndex=$System.NormalizedLoopAge
      Scale Curve=120
     )
     Z=&FloatFromCurve(
      FloatCurve=DataInterface() {propertyValues {Curve {keys [
          {arriveTangent=-1.8682, time=0, value=1}
          {arriveTangent=-2.0886, interpMode=RCIM_Cubic, leaveTangent=-2.0886, tangentMode=RCTM_User, time=1, value=0}
         ]}}}
      CurveIndex=$System.NormalizedLoopAge
      Scale Curve=200
     )
    )) {isSetParameters=true}
   &ScaleMeshSize(Scale Factor=&VectorFromCurve(
     VectorCurve=DataInterface() {propertyValues {
       XCurve {keys [
         {arriveTangent=1.1533, interpMode=RCIM_Cubic, leaveTangent=1.1533, tangentMode=RCTM_User, time=0, value=0.3}
         {arriveTangent=0.2322, interpMode=RCIM_Cubic, leaveTangent=0.2322, tangentMode=RCTM_User, time=1, value=1}
        ]}
       YCurve {keys [
         {arriveTangent=1.1533, interpMode=RCIM_Cubic, leaveTangent=1.1533, tangentMode=RCTM_User, time=0, value=0.3}
         {arriveTangent=0.2322, interpMode=RCIM_Cubic, leaveTangent=0.2322, tangentMode=RCTM_User, time=1, value=1}
        ]}
       ZCurve {keys [
         {arriveTangent=2.4044, interpMode=RCIM_Cubic, time=0, value=0.1}
         {arriveTangent=0.8008, interpMode=RCIM_Cubic, leaveTangent=0.8008, tangentMode=RCTM_User, time=1, value=1}
        ]}
      }}
     CurveIndex=$System.NormalizedLoopAge
    ))
   &ScaleColor(
    Scale Alpha=&FloatFromCurve(FloatCurve=DataInterface() {propertyValues {Curve {keys [
         {arriveTangent=8.125, interpMode=RCIM_Cubic, time=0, value=0}
         {arriveTangentWeight=0.4805, interpMode=RCIM_Cubic, leaveTangentWeight=0.1363, tangentMode=RCTM_User, time=0.8, value=1}
         {interpMode=RCIM_Cubic, leaveTangent=-2.3395, time=1, value=0}
        ]}}})
    Scale RGB=&MakeVectorFromLinearColorRGB(Color=&ColorFromCurve(ColorCurve=DataInterface() {propertyValues {
        AlphaCurve {keys []}
        BlueCurve {keys [
          {arriveTangent=-19.0486, interpMode=RCIM_Cubic, time=0.3389, value=1}
          {arriveTangent=-0.5357, interpMode=RCIM_Cubic, leaveTangent=-0.5357, time=0.8297, value=12}
          {arriveTangent=-155.1807, interpMode=RCIM_Cubic, leaveTangent=-155.1807, tangentMode=RCTM_User, time=1, value=0.6458}
         ]}
        GreenCurve {keys [
          {arriveTangent=-2.2041, interpMode=RCIM_Cubic, time=0.3389, value=0.2185}
          {arriveTangent=8.5346, interpMode=RCIM_Cubic, leaveTangent=8.5346, tangentMode=RCTM_User, time=0.8297, value=2.036}
          {arriveTangent=-44.728, interpMode=RCIM_Cubic, leaveTangent=-44.728, tangentMode=RCTM_User, time=1, value=0.3985}
         ]}
        RedCurve {keys [
          {arriveTangent=-16.8034, interpMode=RCIM_Cubic, time=0.3389, value=0.3345}
          {arriveTangent=0.1443, interpMode=RCIM_Cubic, leaveTangent=0.1443, time=0.8297, value=10.3059}
          {arriveTangent=-155.1807, interpMode=RCIM_Cubic, leaveTangent=-155.1807, tangentMode=RCTM_User, time=1, value=0.4299}
         ]}
       }}))
    ScaleA=true
    ScaleRGB=true
   )
   &DynamicMaterialParameters(
    Index 0 Param 1=&FloatFromCurve(FloatCurve=DataInterface() {propertyValues {Curve {keys [
         {arriveTangent=-5.84, interpMode=RCIM_Cubic, time=0, value=1}
         <at 0.6 0 RCIM_Cubic>
         {interpMode=RCIM_Cubic, leaveTangent=16.4602, time=1, value=1}
        ]}}})
    Index 0 Param 2=&FloatFromCurve(FloatCurve=DataInterface() {propertyValues {Curve {keys [
         {arriveTangent=2.4843, interpMode=RCIM_Cubic, leaveTangent=2.4843, tangentMode=RCTM_User, time=0, value=0}
         {interpMode=RCIM_Cubic, leaveTangent=-0.8461, time=1, value=1}
        ]}}})
    Index 0 Param 4=&FloatFromCurve(
     FloatCurve=DataInterface() {propertyValues {Curve {keys [
         {arriveTangent=0.2805, interpMode=RCIM_Cubic, time=0, value=1}
         {arriveTangent=2.4494, interpMode=RCIM_Cubic, leaveTangent=2.4494, tangentMode=RCTM_User, time=1, value=2}
        ]}}}
     Scale Curve=&RandomRangeFloat(Minimum=0.6, Maximum=0.8)
    )
    Module.Param0WriteEnabled=true
    Module.Param1WriteEnabled=true
    Module.Param2WriteEnabled=false
    Module.Param3WriteEnabled=true
   )
  ]
  rendererClasses [&NiagaraMeshRendererProperties]
  renderers [{rendererClass=&NiagaraMeshRendererProperties, rendererIndex=0}]
 }
}
summary {
 systemName=NS_ShockwaveMace_Slam_Swing
 emitters [
  {
   bEnabled=false
   emitterName=Empty
   rendererClasses [&NiagaraSpriteRendererProperties, &NiagaraRibbonRendererProperties]
  }
  {emitterName=Wind, rendererClasses [&NiagaraMeshRendererProperties]}
  {emitterName=Glow, rendererClasses [&NiagaraSpriteRendererProperties, &NiagaraLightRendererProperties]}
 ]
 userVariables [User.NewNiagaraDataInterfaceSkeletalMesh {
   type=NiagaraDataInterfaceSkeletalMesh
   defaultValue {
    struct=NiagaraExt_VariableValue_DataInterface
    value {dataInterface=&self:NiagaraDataInterfaceSkeletalMesh_0}
   }
  }]
}
