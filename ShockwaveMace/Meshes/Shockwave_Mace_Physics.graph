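# Physics asset for the Shockwave Mace that defines collision bodies and joints for a dynamic cloth piece on the weapon.
# A capsule body sits on each of four bones: Dyn_Master_Cloth and Dyn_Cloth1 through Dyn_Cloth3.
# Constraints link them in a single chain from the master bone outward, with wider swing and twist limits at the first joint than at the two below it.
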PhysicsAsset /ShockwaveMace/Meshes/Shockwave_Mace_Physics
bodies {
 Dyn_Cloth1 [{shapeType=Capsule, length=6, radius=2, shapeName=Dyn_Cloth1_capsule, center=<4.1666 -0.7725 -0.2726>, rotation=<rot 90 0 0>}]
 Dyn_Cloth2 [{shapeType=Capsule, length=4.25, radius=3, shapeName=Dyn_Cloth2_capsule, center=<2.5 0 0>, rotation=<rot 90 0 0>}]
 Dyn_Cloth3 [{shapeType=Capsule, length=4.25, radius=3, shapeName=Dyn_Cloth3_capsule, center=<2.5 0 0>, rotation=<rot 90 0 0>}]
 Dyn_Master_Cloth [{shapeType=Capsule, length=2.25, radius=5.5, shapeName=Dyn_Master_Cloth_capsule, center=<-2.9016 4.5602 -1.1077>, rotation=<rot 0 -90 0>}]
}
constraints [
 {bone1Name=Dyn_Cloth1, bone2Name=Dyn_Master_Cloth, swing1LimitDegrees=90, swing2LimitDegrees=90, twistLimitDegrees=90}
 {bone1Name=Dyn_Cloth2, bone2Name=Dyn_Cloth1, swing1LimitDegrees=60, swing2LimitDegrees=60, twistLimitDegrees=45}
 {bone1Name=Dyn_Cloth3, bone2Name=Dyn_Cloth2, swing1LimitDegrees=60, swing2LimitDegrees=60, twistLimitDegrees=45}
]
