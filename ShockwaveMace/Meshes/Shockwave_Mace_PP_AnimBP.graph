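# Post-process AnimBP for the Shockwave Mace that plays its pop-out animation when the weapon is out of charge.
# IsCharged mirrors magazine ammo above zero, driving an Idle, Intro, Loop, Outro state machine.
# That machine is layered over the input pose from core_mstr, then fed through RigidBody simulation relative to the handle bone.
# After evaluation, a zero-delay step writes the Dyn_Master_Core socket location to material parameter VP_SocketLocation.
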AnimBlueprint /ShockwaveMace/Meshes/Shockwave_Mace_PP_AnimBP
target_skeleton=&/Shockwave_Mace_Skeleton
graphs [
 AnimGraph {nodes [
   n1 = Input Pose()
   AnimGraphNode_StateMachine_0 = PopOut State Machine() {is=title AnimGraphNode_StateMachine}
   n2 = RigidBody(ComponentPose=Local To Component(LocalPose=Layered blend per bone(BasePose=@n1, BlendPoses_0=@AnimGraphNode_StateMachine_0) {layerSetup [{branchFilters [{boneName=core_mstr}]}]})) {
    bClampLinearTranslationLimitToRefPose=true
    bFreezeIncomingPoseOnStart=true
    bOverrideWorldGravity=true
    overridePhysicsAsset=&/Shockwave_Mace_Physics
    simulationSpace=BaseBoneSpace
    alphaBoolBlend {blendOption=Linear}
    baseBoneRef {boneName=Melee_TwoHanded_Handle}
    componentAppliedLinearAccClamp=<4000 4000 4000>
    componentLinearAccScale=<2 2 2>
    componentLinearVelScale=<2 2 2>
    overrideWorldGravity {z=-980}
   }
   Output Pose(Result=Component To Local(ComponentPose=@n2)) {layerGroup=None}
  ]}
 PopOut State Machine {
  entry=IdlePose
  path=AnimGraph.AnimGraphNode_StateMachine_0.PopOut State Machine
  ids=id
  nodes [
   AnimStateNode_0 = IdlePose() {is=title AnimStateNode}
   AnimStateNode_1 = PopOut Intro() {is=title AnimStateNode}
   AnimStateNode_2 = PopOut Loop() {is=title AnimStateNode}
   AnimStateNode_3 = PopOut Outro() {is=title AnimStateNode}
   AnimStateTransitionNode_0 = AnimStateTransitionNode() {crossfadeDuration=0.1}
   AnimStateTransitionNode_1 = AnimStateTransitionNode() {bAutomaticRuleBasedOnSequencePlayerInState=true, crossfadeDuration=0.1}
   AnimStateTransitionNode_2 = AnimStateTransitionNode() {crossfadeDuration=0.1}
   AnimStateTransitionNode_3 = AnimStateTransitionNode() {bAutomaticRuleBasedOnSequencePlayerInState=true}
  ]
  links [
   {
    dst=PopOut Intro
    rule=AnimGraph.AnimGraphNode_StateMachine_0.PopOut State Machine.AnimStateTransitionNode_0.Transition
    src=IdlePose
    via=AnimStateTransitionNode_0
   }
   {
    dst=PopOut Loop
    rule=AnimGraph.AnimGraphNode_StateMachine_0.PopOut State Machine.AnimStateTransitionNode_1.Transition
    src=PopOut Intro
    via=AnimStateTransitionNode_1
   }
   {
    dst=PopOut Outro
    rule=AnimGraph.AnimGraphNode_StateMachine_0.PopOut State Machine.AnimStateTransitionNode_2.Transition
    src=PopOut Loop
    via=AnimStateTransitionNode_2
   }
   {
    dst=IdlePose
    rule=AnimGraph.AnimGraphNode_StateMachine_0.PopOut State Machine.AnimStateTransitionNode_3.Transition
    src=PopOut Outro
    via=AnimStateTransitionNode_3
   }
  ]
 }
 IdlePose {
  path=AnimGraph.AnimGraphNode_StateMachine_0.PopOut State Machine.AnimStateNode_0.IdlePose
  nodes [Output Animation Pose(Result=Sequence Player 'ShockWave_Mace_IdlePose_Shockwave_Mace_Melee'() {sequence=/ShockwaveMace/Anim/BasicHammer/Mace/ShockWave_Mace_IdlePose_Shockwave_Mace_Melee})]
 }
 PopOut Intro {
  path=AnimGraph.AnimGraphNode_StateMachine_0.PopOut State Machine.AnimStateNode_1.PopOut Intro
  nodes [Output Animation Pose(Result=Sequence Player 'ShockWave_Mace_PopOut_Start_Shockwave_Mace_Melee'() {bLoopAnimation=false, sequence=/ShockwaveMace/Anim/BasicHammer/Mace/ShockWave_Mace_PopOut_Start_Shockwave_Mace_Melee})]
 }
 PopOut Loop {
  path=AnimGraph.AnimGraphNode_StateMachine_0.PopOut State Machine.AnimStateNode_2.PopOut Loop
  nodes [Output Animation Pose(Result=Sequence Player 'ShockWave_Mace_PopOut_Loop_Shockwave_Mace_Melee'() {sequence=/ShockwaveMace/Anim/BasicHammer/Mace/ShockWave_Mace_PopOut_Loop_Shockwave_Mace_Melee})]
 }
 PopOut Outro {
  path=AnimGraph.AnimGraphNode_StateMachine_0.PopOut State Machine.AnimStateNode_3.PopOut Outro
  nodes [Output Animation Pose(Result=Sequence Player 'ShockWave_Mace_PopOut_End_Shockwave_Mace_Melee'() {bLoopAnimation=false, sequence=/ShockwaveMace/Anim/BasicHammer/Mace/ShockWave_Mace_PopOut_End_Shockwave_Mace_Melee})]
 }
 Transition {
  path=AnimGraph.AnimGraphNode_StateMachine_0.PopOut State Machine.AnimStateTransitionNode_0.Transition
  nodes [Result(bCanEnterTransition=NOT Boolean(A=Get IsCharged() -> (IsCharged:bool)))]
 }
 Transition {path=AnimGraph.AnimGraphNode_StateMachine_0.PopOut State Machine.AnimStateTransitionNode_1.Transition, nodes [Result()]}
 Transition {
  path=AnimGraph.AnimGraphNode_StateMachine_0.PopOut State Machine.AnimStateTransitionNode_2.Transition
  nodes [Result(bCanEnterTransition=Get IsCharged())]
 }
 Transition {path=AnimGraph.AnimGraphNode_StateMachine_0.PopOut State Machine.AnimStateTransitionNode_3.Transition, nodes [Result()]}
 EventGraph {nodes [
   Event BlueprintUpdateAnimation()
   n1 = GetOwningActor()
   > Is Valid(InputObject=@n1)
   > n2 = Cast To B_ShockwaveMace!(execute, Object=@n1) -> (then, CastFailed, AsB Shockwave Mace:B Shockwave Mace Object Reference) {is=title K2Node_DynamicCast}
   > Set IsCharged(IsCharged:bool="integer > integer"(A=GetMagazineAmmoCount!(self:Fort Weapon Object Reference=@n2) -> (ReturnValue:int=0))) -> (Output_Get:bool)
   Event BlueprintPostEvaluateAnimation() {is=title K2Node_Event}
   n3 = GetPartSkeletalMeshComponent()
   > Is Valid(InputObject=@n3)
   > Delay(Duration=0)
   > SetVectorParameterValueOnMaterials!(
    execute
    self:Mesh Component Object Reference=@n3
    ParameterName:Name=VP_SocketLocation
    ParameterValue:Vector=GetSocketLocation!(self:Scene Component Object Reference=@n3, InSocketName:Name=Dyn_Master_Core) -> (ReturnValue:Vector)
   ) -> (then)
  ]}
]
variables [IsCharged {type=boolean}]
